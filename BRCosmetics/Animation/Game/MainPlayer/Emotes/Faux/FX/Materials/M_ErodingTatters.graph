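# Masked, two-sided lit Niagara mesh-particle material for eroding cloth tatters in the Faux emote.
# A sub-UV Tatters texture supplies erosion, shape and noise channels; particle alpha drives erosion, with a contrast-broken, eye-adapted edge glow.
# Faux body textures provide shading, sampled as small patches offset by dynamic RandU on both axes, with diffuse tinted by particle color.
# On Epic quality, vertices rotate around the particle center by Bend scaled across U, with normals corrected.
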Material /BRCosmetics/Animation/Game/MainPlayer/Emotes/Faux/FX/Materials/M_ErodingTatters
graph {
 properties {
  bCastRayTracedShadows=false
  bUsedWithNiagaraMeshParticles=true
  blendMode=BLEND_Masked
  pixelDepthOffsetMode=PDOM_Legacy
  twoSided=true
  usedShadingModels=Default Lit
 }
 nodes [
  ComponentMask() {g=true, r=true}
  MaterialFunctionCall(
   Rotation Axis=NamedRerouteUsage()
   Rotation Angle=NamedRerouteUsage()
   World Space Vertex Normals=TextureSample(UVs=NamedRerouteUsage()) {samplerType=SAMPLERTYPE_Normal, texture=/BRCosmetics/Characters/Player/Male/Medium/Bodies/M_MED_Faux/Textures/T_Faux_body_N}.RGB
  ) {materialFunction=&FixRotateAboutAxisNormals}
  n1 = DynamicParameter() {parameterIndex=1, defaultValue=<rgba 1 1 0.1 0.2>, paramNames [R,G,B,A]}
  n2 = ParticleColor()
  n3 = LinearInterpolate() {constA=0.01, constB=0.1}
  n4 = Subtract(A=NamedRerouteUsage(), B=Add(A=OneMinus(None=@n2.A), B=@n3) {constB=0.1})
  n5 = Multiply(A=Abs(None=Min(A=@n4) {constB=0}), B=Divide(B=@n3))
  n6 = MaterialFunctionCall(In=NamedRerouteUsage(), Contrast=Constant() {r=0.5}) {materialFunction=&CheapContrast}
  Multiply(
   A=Multiply(
    A=MaterialFunctionCall(R=@n1.R, G=@n1.R, B=@n1.R) {materialFunction=&AppendMany}.RGB
    B=Multiply(A=@n5, B=@n6.Result)
   )
   B=EyeAdaptation().EyeAdaptation
  )
  n7 = ParticleSubUV!(UVs, Tex, Apply View MipBias) {texture=/BRCosmetics/Animation/Game/MainPlayer/Emotes/Faux/FX/Textures/Tatters}
  n8 = Subtract(A=NamedRerouteDeclaration(None=@n7.R) {name=ErrosionTexture, nodeColor=<rgba 0 1 0.0859 1>}, B=OneMinus(None=ParticleColor().A))
  Multiply(A=Ceil(None=Saturate(None=@n8)), B=@n7.G)
  n9 = TextureSample!(UVs=NamedRerouteUsage(), Tex, Level, Apply View MipBias) {
   mipValueMode=TMVM_MipLevel
   samplerSource=SSM_Clamp_WorldGroupSettings
   texture=/BRCosmetics/Characters/Player/Male/Medium/Bodies/M_MED_Faux/Textures/T_Faux_body_D
  }
  Multiply(A=ParticleColor().RGB, B=@n9.RGB)
  n10 = TextureCoordinate()
  n11 = Constant() {r=0.1}
  n12 = DynamicParameter() {defaultValue=<rgba 1 1 0.1 0.2>, paramNames [Bend,FallOff,RandU,RandV]}
  NamedRerouteDeclaration(None=Add(A=Multiply(A=@n10, B=@n11), B=AppendVector(A=@n12.RandU, B=@n12.RandU))) {name=TextureUVs, nodeColor=<rgba 1 0 0.8672 1>}
  n13 = DynamicParameter() {defaultValue=<rgba 1 1 0.1 0.2>, paramNames [Bend,FallOff,RandU,RandV]}
  n14 = Subtract(A=MaterialFunctionCall(Float2=TextureCoordinate()) {materialFunction=&BreakOutFloat2Components}.R) {constB=0.5}
  n15 = Sign(None=@n14)
  n16 = Multiply(A=@n14) {constB=2}
  NamedRerouteDeclaration(None=Multiply(A=Multiply(A=@n13.Bend, B=@n15), B=Abs(None=@n16))) {name=RotationAngle, nodeColor=<rgba 0 1 0.1562 1>}
  NamedRerouteDeclaration(None=@n7.B) {name=NoiseTexture, nodeColor=<rgba 0 0.7422 1 1>}
  ParticleRelativeTime()
  n17 = Constant3Vector() {constant=<rgba 0 0 1 0>}
  n18 = NamedRerouteDeclaration(None=Normalize(VectorInput=Transform(None=@n17) {transformSourceType=TRANSFORMSOURCE_World, transformType=TRANSFORM_ParticleWorld})) {name=RotationAxis, nodeColor=<rgba 0 0.0156 1 1>}
  n19 = NamedRerouteUsage()
  n20 = ParticlePositionWS()
  n21 = WorldPosition() {worldPositionShaderOffset=WPT_ExcludeAllShaderOffsets}
  QualitySwitch(
   Default=Constant3Vector()
   Epic=MaterialFunctionCall!(InVector=RotateAboutAxis(NormalizedRotationAxis=@n18, RotationAngle=@n19, PivotPoint=@n20, Position=@n21.XYZ)) {materialFunction=/Game/Athena/Environments/Landscape/Foliage/MF_MakePreciseVec3}.Result
  )
  TextureSample!(UVs=NamedRerouteUsage(), Tex, Apply View MipBias) {samplerType=SAMPLERTYPE_Masks, texture=/BRCosmetics/Characters/Player/Male/Medium/Bodies/M_MED_Faux/Textures/T_Faux_body_S}
  TextureSample(UVs=NamedRerouteUsage()) {samplerType=SAMPLERTYPE_Masks, texture=/BRCosmetics/Characters/Player/Male/Medium/Bodies/M_MED_Faux/Textures/T_Faux_body_M}
  Transform(None=@n17) {transformSourceType=TRANSFORMSOURCE_World, transformType=TRANSFORM_Local}
 ]
}
