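# Masked, two-sided lit material for Niagara sprites in the Faux emote FX, producing tattered pieces of the Faux character's body that erode away.
# Particle alpha acts as the erosion threshold: one minus alpha is subtracted from the Tatters red channel, then saturated, ceiled to a hard edge and multiplied by green.
# Particle color tints the body diffuse texture, which shares rerouted UVs with the body S and M texture samples.
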Material /BRCosmetics/Animation/Game/MainPlayer/Emotes/Faux/FX/Materials/M_Errosion
graph {
 properties {
  bCastRayTracedShadows=false
  bUsedWithNiagaraSprites=true
  blendMode=BLEND_Masked
  pixelDepthOffsetMode=PDOM_Legacy
  twoSided=true
  usedShadingModels=Default Lit
 }
 nodes [
  MaterialFunctionCall!(In, Ramp Width) {materialFunction=/Engine/Functions/Engine_MaterialFunctions01/Math/SmoothCeil}
  n1 = TextureSample!(UVs, Tex, Level, Apply View MipBias) {
   mipValueMode=TMVM_MipLevel
   samplerSource=SSM_Clamp_WorldGroupSettings
   texture=/BRCosmetics/Animation/Game/MainPlayer/Emotes/Faux/FX/Textures/Tatters
  }
  n2 = Subtract(A=@n1.R, B=OneMinus(None=ParticleColor().A))
  Multiply(A=Ceil(None=Saturate(None=@n2)), B=@n1.G)
  n3 = TextureSample(UVs=NamedRerouteUsage()) {
   mipValueMode=TMVM_MipLevel
   samplerSource=SSM_Clamp_WorldGroupSettings
   texture=/BRCosmetics/Characters/Player/Male/Medium/Bodies/M_MED_Faux/Textures/T_Faux_body_D
  }
  Multiply(A=ParticleColor().RGB, B=@n3.RGB)
  ParticleColor()
  Saturate(None=@n1.R)
  TextureSample!(UVs=NamedRerouteUsage(), Tex, Apply View MipBias) {samplerType=SAMPLERTYPE_Masks, texture=/BRCosmetics/Characters/Player/Male/Medium/Bodies/M_MED_Faux/Textures/T_Faux_body_S}
  TextureSample(UVs=NamedRerouteUsage()) {samplerType=SAMPLERTYPE_Masks, texture=/BRCosmetics/Characters/Player/Male/Medium/Bodies/M_MED_Faux/Textures/T_Faux_body_M}
 ]
}
