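# Masked, two-sided subsurface material for Niagara mesh particles, apparently the paint spray in the Faux emote.
# Two water-foam height and normal samples, offset by dynamic parameters, blend along V through a sine.
# The blended mask combines with contrast and side-edge masks for shaping, and drives brightness over a three-color gradient across U.
# Other branches build panned distortion offsets for WPO and tints from the Faux body texture and particle color.
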Material /BRCosmetics/Animation/Game/MainPlayer/Emotes/Faux/FX/Materials/M_Paint_Spray
graph {
 properties {
  bUsedWithNiagaraMeshParticles=true
  blendMode=BLEND_Masked
  pixelDepthOffsetMode=PDOM_Legacy
  shadingModel=MSM_Subsurface
  twoSided=true
  usedShadingModels=Subsurface
 }
 nodes [
  n1 = TextureCoordinate()
  n2 = Constant()
  n3 = TextureSampleParameter2D(UVs=NamedRerouteUsage()) {
   group=Mask
   parameterName=Mask
   samplerType=SAMPLERTYPE_LinearColor
   texture=/Game/Athena/Apollo/Water/Textures/02_Rounded/02_WaterFoam_Veroni_base_H
  }
  n4 = TextureSampleParameter2D(UVs=NamedRerouteUsage()) {
   group=Mask
   parameterName=Mask
   samplerType=SAMPLERTYPE_LinearColor
   texture=&02_WaterFoam_Veroni_base_H
  }
  n5 = ScalarParameter() {defaultValue=0.75, group=UV, parameterName=Blend Masks Phase}
  n6 = ComponentMask(None=TextureCoordinate()) {g=true}
  n7 = OneMinus(None=MaterialFunctionCall(Value=Add(A=@n5, B=@n6), Period=ScalarParameter() {defaultValue=1, group=UV, parameterName=Blend Masks Period}) {materialFunction=&LinearSine}.Linear Sine)
  n8 = ScalarParameter() {defaultValue=-0.25, group=UV, parameterName=Texture Mask Offset}
  n9 = LinearInterpolate(A=@n3.G, B=@n4.G, Alpha=NamedRerouteDeclaration(None=Add(A=@n7, B=@n8)) {name=Mask Blend Alpha, nodeColor=<rgba 0.4062 1 0 1>})
  Add(
   A=DynamicParameter() {defaultValue=<rgba 1 0.5 5 0>, paramNames [Offset V, Offset V2, GradientSoftness, UVOffset]}.UVOffset
   B=Add(A=AppendVector(A=@n1, B=@n2), B=Multiply(A=@n9) {constB=0.2})
  )
  n10 = TextureCoordinate()
  n11 = ComponentMask(None=@n10) {r=true}
  n12 = ScalarParameter() {defaultValue=0.1, group=WPO, parameterName=WPO Tile U}
  n13 = ComponentMask(None=@n10) {g=true}
  n14 = ScalarParameter() {defaultValue=7, group=WPO, parameterName=WPO Tile V}
  n15 = Panner(
   Coordinate=AppendVector(A=Multiply(A=@n11, B=@n12), B=Multiply(A=@n13, B=@n14))
   Time=Time() {bOverride_Period=true, period=64}
   Speed=AppendVector(
    A=ScalarParameter() {defaultValue=0.1, group=WPO, parameterName=WPO Pan Speed U}
    B=ScalarParameter() {defaultValue=0.4, group=WPO, parameterName=WPO Pan Speed V}
   )
  ) {speedX=-1, speedY=0.1}
  n16 = TextureSample(UVs=@n15) {samplerType=SAMPLERTYPE_Normal, texture=&T_Weap_Distort_N_02}
  n17 = ScalarParameter() {defaultValue=16.3025, group=WPO, parameterName=WPO Amnt}
  AppendVector(
   A=Multiply(A=AppendVector(A=@n16.R, B=@n16.R), B=Multiply(B=@n17) {constA=0.5})
   B=Multiply(A=TextureSample(UVs=@n15) {samplerType=SAMPLERTYPE_Normal, texture=/Game/Effects/Fort_Effects/Textures/T_DustDistortion}.R, B=@n17)
  )
  Constant() {r=0.2}
  Constant() {r=1}
  n18 = TextureSample(UVs=NamedRerouteUsage()) {texture=/BRCosmetics/Characters/Player/Male/Medium/Bodies/M_MED_Faux/Textures/T_Faux_body_D}
  n19 = ParticleColor()
  n20 = LinearInterpolate(Alpha=Power(Base=@n9) {constExponent=0.7}) {constB=0.15}
  n21 = Add(A=LinearInterpolate(A=@n18.RGB, B=@n19.RGB, Alpha=ScalarParameter() {group=Color, parameterName=Use Particle Color}), B=@n20)
  n22 = ScalarParameter() {defaultValue=0.1, group=Color, parameterName=EmissiveAdd}
  n23 = Power(Base=@n9) {constExponent=0.9}
  MaterialFunctionCall(Emissive=Add(A=Multiply(A=@n21, B=@n22) {constB=0.15}, B=LinearInterpolate(Alpha=@n23) {constB=0.025})) {materialFunction=&MF_FX_EmissiveAdjustments}
  MaterialFunctionCall(Hue Shift Percentage=Constant() {r=0.1}, Texture=Multiply(A=@n18.RGB, B=@n19.RGB)) {materialFunction=&HueShift}
  n24 = ComponentMask(None=TextureCoordinate()) {g=true}
  n25 = ScalarParameter() {group=Mask, parameterName=Main Mask Contrast}
  n26 = TextureCoordinate()
  n27 = Constant() {r=0.5}
  n28 = Multiply(
   A=Add(
    A=MaterialFunctionCall(In=@n24, Contrast=@n25) {materialFunction=&CheapContrast}.Result
    B=ScalarParameter() {defaultValue=-0.5, group=Mask, parameterName=Main Mask Offset}
   )
   B=SphereMask(A=ComponentMask(None=@n26) {r=true}, B=@n27, Radius=@n27, Hardness=ScalarParameter() {group=Mask, parameterName=Side Edge Hardness}) {hardnessPercent=1}
  )
  n29 = OneMinus(None=Multiply(A=OneMinus(None=@n9), B=OneMinus(None=@n28)))
  Multiply(A=VertexColor().R, B=@n29)
  Multiply(A=ComponentMask(None=TextureCoordinate()) {g=true}, B=@n29)
  n30 = ComponentMask(None=TextureCoordinate()) {g=true}
  n31 = TextureSampleParameter2D(UVs=NamedRerouteUsage()) {
   group=Normal
   parameterName=Normal
   samplerType=SAMPLERTYPE_Normal
   texture=/Game/Athena/Apollo/Water/Textures/02_Rounded/02_WaterFoam_Veroni_base_N
  }
  n32 = TextureSampleParameter2D(UVs=NamedRerouteUsage()) {
   group=Normal
   parameterName=Normal
   samplerType=SAMPLERTYPE_Normal
   texture=&02_WaterFoam_Veroni_base_N
  }
  n33 = LinearInterpolate(A=@n31.RGB, B=@n32.RGB, Alpha=NamedRerouteUsage())
  n34 = Constant3Vector() {constant=<rgba 0 0 1 0>}
  n35 = ScalarParameter() {defaultValue=0.2, group=Normal, parameterName=Flatten Normal}
  Multiply(A=Multiply(A=Multiply(A=@n30, B=@n30), B=LinearInterpolate(A=@n33, B=@n34, Alpha=@n35))) {constB=3}
  n36 = VectorParameter() {group=Color, parameterName=Color_Mid, defaultValue=<rgba 1 0.1591 0.538 1>}
  n37 = VectorParameter() {group=Color, parameterName=Color_Left, defaultValue=<rgba 0.1967 0.1696 1 1>}
  n38 = ComponentMask(None=TextureCoordinate()) {r=true}
  n39 = MaterialFunctionCall(
   In=Add(
    A=NamedRerouteDeclaration(None=@n38) {name=U, nodeColor=<rgba 0 1 0.4609 1>}
    B=ScalarParameter() {defaultValue=-0.2, group=Color, parameterName=Color1 Offset}
   )
   Contrast=ScalarParameter() {defaultValue=10, group=Color, parameterName=Color1 Contrast}
  ) {materialFunction=&CheapContrast}
  n40 = Add(A=NamedRerouteUsage(), B=ScalarParameter() {defaultValue=0.2, group=Color, parameterName=Color2 Offset})
  n41 = ScalarParameter() {defaultValue=10, group=Color, parameterName=Color2 Contrast}
  Multiply(
   A=LinearInterpolate(
    A=VectorParameter() {group=Color, parameterName=Color_Right, defaultValue=<rgba 0.099 1 0.7892 1>}.RGB
    B=LinearInterpolate(A=@n36.RGB, B=@n37.RGB, Alpha=@n39.Result)
    Alpha=MaterialFunctionCall(In=@n40, Contrast=@n41) {materialFunction=&CheapContrast}.Result
   )
   B=@n20
  )
  n42 = VectorParameter() {group=UV, parameterName=Mask 1 Tile, defaultValue=<rgba 0.5 0.275 0 0>}
  n43 = Multiply(A=AppendVector(A=@n42.R, B=@n42.R), B=TextureCoordinate())
  n44 = MaterialFunctionCall(Float2=@n43) {materialFunction=&BreakOutFloat2Components}
  n45 = DynamicParameter() {defaultValue=<rgba 1 0.5 5 0>, paramNames [Offset V, Offset V2, GradientSoftness, UVOffset]}
  NamedRerouteDeclaration(None=AppendVector(A=@n44.R, B=Add(A=@n44.G, B=@n45.Offset V))) {name=Panning UV Mask 1, nodeColor=<rgba 1 0 0.7266 1>}
  n46 = VectorParameter() {group=UV, parameterName=Mask 2 Tile, defaultValue=<rgba 1 0.55 0 0>}
  n47 = Multiply(A=AppendVector(A=@n46.R, B=@n46.R), B=TextureCoordinate())
  n48 = MaterialFunctionCall(Float2=@n47) {materialFunction=&BreakOutFloat2Components}
  n49 = DynamicParameter() {defaultValue=<rgba 1 0.5 5 0>, paramNames [Offset V, Offset V2, GradientSoftness, UVOffset]}
  NamedRerouteDeclaration(None=AppendVector(A=@n48.R, B=Add(A=@n48.G, B=@n49.Offset V2))) {name=Panning UV Mask 2, nodeColor=<rgba 1 0 0.3984 1>}
  ObjectPositionWS()
  ObjectPositionWS()
  n50 = Add(A=Add(A=Time(), B=ScalarParameter() {group=UV, parameterName=UV Offset}), B=ComponentMask(None=ParticlePositionWS()) {g=true})
  n51 = ScalarParameter() {defaultValue=-1.3, group=UV, parameterName=Mask 1 Pan Speed}
  n52 = TextureCoordinate() {uTiling=0.9, vTiling=0.9}
  n53 = AppendVector(
   A=Multiply(A=ComponentMask(None=@n52) {r=true}, B=ScalarParameter() {defaultValue=2.5, group=Distortion, parameterName=Distort Tile U})
   B=Multiply(A=ComponentMask(None=@n52) {g=true}, B=ScalarParameter() {defaultValue=1, group=Distortion, parameterName=Distort Tile V})
  )
  n54 = Add(A=Time(), B=ScalarParameter() {group=UV, parameterName=UV Offset})
  n55 = AppendVector(A=ScalarParameter() {defaultValue=0.05, parameterName=Distort Pan U}, B=ScalarParameter() {defaultValue=0.43, parameterName=Distort Pan V})
  n56 = Multiply(
   A=ScalarParameter() {defaultValue=0.2, group=Distortion, parameterName=Distort Strenght}
   B=TextureSample(UVs=Panner(Coordinate=@n53, Time=@n54, Speed=@n55)) {samplerType=SAMPLERTYPE_LinearColor, texture=&02_WaterFoam_Veroni_base_H}.R
  ) {constB=2}
  Panner(Coordinate=@n43, Time=Add(A=Multiply(A=@n50, B=@n51), B=@n56)) {speedX=-0.05, speedY=-0.3}
  n57 = Add(A=Add(A=Time(), B=ScalarParameter() {group=UV, parameterName=UV Offset}), B=ComponentMask(None=ParticlePositionWS()) {g=true})
  n58 = ScalarParameter() {defaultValue=-3.5, group=UV, parameterName=Mask 2 Pan Speed}
  Panner(Coordinate=@n47, Time=Add(A=Multiply(A=@n57, B=@n58), B=@n56)) {speedX=0.05, speedY=-0.16}
  TextureCoordinate()
  TextureCoordinate()
  TextureCoordinate()
  TextureSample(UVs=NamedRerouteUsage()) {texture=&T_Inky_Smoke_Tile}
  VectorParameter() {group=Color, parameterName=Subsurface Color, defaultValue=<rgba 0.349 0.0915 0.1711 1>}
 ]
}
