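# Additive, unlit, two-sided material for Niagara mesh particles, apparently the energy spray effect in the Faux player emote.
# Particle color tints a soft caustics texture, which is multiplied by a sharp cellular-noise mask stretched along V.
# The mask comes from a high threshold that varies slightly along V, then is steeply scaled and saturated.
# Dynamic parameters drive tiling and V offset for both textures, and MF_FX_EmissiveAdjustments processes the emissive result.
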Material /BRCosmetics/Animation/Game/MainPlayer/Emotes/Faux/FX/Materials/M_SprayEnergy
graph {
 properties {
  bUsedWithNiagaraMeshParticles=true
  blendMode=BLEND_Additive
  pixelDepthOffsetMode=PDOM_Legacy
  shadingModel=MSM_Unlit
  twoSided=true
  usedShadingModels=Unlit
 }
 nodes [
  n1 = ParticleColor()
  n2 = TextureCoordinate()
  n3 = DynamicParameter() {defaultValue=<rgba 1 0 1 0>, paramNames [Caustic Tile, Caustic Offset, Cell Noise Mask Tile, Cell Noise Offset]}
  n4 = MaterialFunctionCall(Float2=Multiply(A=@n2, B=@n3.Caustic Tile)) {materialFunction=&BreakOutFloat2Components}
  n5 = TextureSample(UVs=AppendVector(A=@n4.R, B=Add(A=@n4.G, B=@n3.Caustic Offset))) {texture=/Game/Global/Textures/Default/Caustics/T_Caustics_Softer}
  n6 = TextureCoordinate()
  n7 = Constant2Vector() {g=0.05, r=1}
  n8 = DynamicParameter() {defaultValue=<rgba 1 0 1 0>, paramNames [Caustic Tile, Caustic Offset, Cell Noise Mask Tile, Cell Noise Offset]}
  n9 = MaterialFunctionCall(Float2=Multiply(A=Multiply(A=@n6, B=@n7), B=@n8.Cell Noise Mask Tile)) {materialFunction=&BreakOutFloat2Components}
  n10 = AppendVector(A=@n9.R, B=Add(A=@n9.G, B=@n8.Cell Noise Offset))
  n11 = ComponentMask(None=TextureCoordinate()) {g=true}
  n12 = Multiply(A=Add(
    A=TextureSample(UVs=@n10) {samplerType=SAMPLERTYPE_Masks, texture=/Game/Global/Textures/Default/Noise/T_CellularNoise}.R
    B=LinearInterpolate(Alpha=@n11) {constA=-0.9, constB=-0.85}
   ) {constB=-0.85}) {constB=20}
  MaterialFunctionCall(Emissive=Multiply(A=Multiply(A=@n1.RGB, B=@n5.R), B=Saturate(None=@n12))) {materialFunction=&MF_FX_EmissiveAdjustments}
  Multiply(A=@n2)
  TextureCoordinate()
 ]
}
