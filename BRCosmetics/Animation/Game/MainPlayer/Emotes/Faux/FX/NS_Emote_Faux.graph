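# Niagara effect for the Faux player emote: GPU tumbling sprite and mesh particles with bright blue material tint, plus soft smoke sprites.
# User.NormalizedNotifyProgress, likely fed by an animation notify, drives both emitters' spawn rate, cylinder radius and descending height, plus particle size and spawn-volume tilt.
# Particles then drift under acceleration, with curl noise on the tumbling bits, whose sprite size also varies by quality level.
# The attached-mesh emitter is disabled.
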NiagaraSystem /BRCosmetics/Animation/Game/MainPlayer/Emotes/Faux/FX/NS_Emote_Faux
dependencies {
 usedDataInterfaces [&NiagaraDataInterfaceCurve, &NiagaraDataInterfaceArrayFloat]
 usedDynamicInputs []
 usedModules [
  &SystemState
  &self:SystemSpawnScript.SystemScriptSource.SystemScriptGraph.NiagaraNodeAssignment_0.SetVariables_#1
 ]
}
emitters {
 Empty {
  simTarget=GPUComputeSim
  emitterUpdateScript [
   &EmitterState()
   &SpawnRate(
    SpawnRate=&Multiply_Float(
     A=150
     B=&FloatFromCurve(
      FloatCurve=DataInterface() {propertyValues {Curve {keys [<at 0.3 -0.025 RCIM_Cubic>, {arriveTangent=0.0357, interpMode=RCIM_Cubic, leaveTangent=0.0357, time=0.4, value=0.975}, <at 1 0>]}}}
      CurveIndex=$User.NormalizedNotifyProgress
     )
    )
    Use Spawn Probability=false
   )
  ]
  particleSpawnScript [
   &InitializeParticle(
    Color Mode=::NewEnumerator0
    Lifetime=&RandomRangeFloat(Minimum=1, Maximum=4)
    Mesh Scale Mode=::NewEnumerator3
    Sprite Size Max=<10 15>
    Sprite Size Mode=::NewEnumerator2
    UsePositionOffset=false
    Write Material Random=true
   )
   /Game/Effects/Niagara/Modules/Math/FN_BiasedRandom/v1/FN_BiasedRandom(Bias=0.85, Distribution Contrast=20)
   /SetVariables(Particles.RandomSize:NiagaraFloat=&Multiply_Float(
     A=&Lerp_Float(A=0.05, B=0.3, Alpha=$StackContext.BiasedRandom)
     B=&Lerp_Float(A=1, B=0.25, Alpha=$User.NormalizedNotifyProgress)
    )) {isSetParameters=true}
   /SetVariables(Particles.Scale=&Multiply_VectorByFloat(
     Vector=$Particles.Scale
     Float=&Multiply_Float(
      A=&Lerp_Float(A=0.05, B=0.3, Alpha=$StackContext.BiasedRandom)
      B=&Lerp_Float(A=1, B=0.25, Alpha=$User.NormalizedNotifyProgress)
     )
    )) {isSetParameters=true}
   /SetVariables(Particles.SpriteSize=&Multiply_Vector2(
     A=$Particles.SpriteSize
     B=&Vector2DFromFloat(Value=&Multiply_Float(A=$Particles.RandomSize, B=4))
    )) {isSetParameters=true}
   &CylinderLocation(
    Coordinate Space:ENiagaraCoordinateSpace=::World
    Cylinder Height=60.4916
    Cylinder Origin:Vector3f=/Niagara/DynamicInputs/Helpers/SimulationPosition()
    Cylinder Radius=&FloatFromCurve(
     FloatCurve=DataInterface() {propertyValues {Curve {keys [{arriveTangent=-0.0803, interpMode=RCIM_Cubic, leaveTangent=-0.0803, tangentMode=RCTM_User, time=0.5, value=0.475}, <at 1 1.025>]}}}
     CurveIndex=$User.NormalizedNotifyProgress
     Scale Curve=40
    )
    Hemisphere X=true
    MaskSpawnGroup=false
    Module.Enable Lathe Profile=false
    Module.Override Randomness=false
    Module.Surface Only=false
    Offset=<0 0 21.7494>
   )
   /SetVariables(Particles.Position=&Add_Vector(
     A=&RotateVectorByQuaternion(
      Vector To Rotate=&Subtract_Vector(
       A=Linked() {linkedVariable {name=Particles.Position, type=NiagaraPosition}}
       B=Linked() {linkedVariable {name=Engine.Owner.Position, type=NiagaraPosition}}
      )
      Quaternion=&MakeQuatFromAxisAngle(
       Axis=$Engine.Owner.SystemYAxis
       AngleInDegrees=&FloatFromCurve(
        FloatCurve=DataInterface() {propertyValues {Curve {keys [<at 0.5 0.5>,<at 1 1>]}}}
        CurveIndex=$User.NormalizedNotifyProgress
        Scale Curve=90
       )
       NormalizeAxis=true
      )
     )
     B=Linked() {linkedVariable {name=Engine.Owner.Position, type=NiagaraPosition}}
    )) {isSetParameters=true}
   /SetVariables(Particles.Position=&Add_Vector(
     A=Linked() {linkedVariable {name=Particles.Position, type=NiagaraPosition}}
     B=&MakeVector(
      X=0
      Y=0
      Z=&FloatFromCurve(
       FloatCurve=DataInterface() {propertyValues {Curve {keys [
           {arriveTangent=-122.6229, interpMode=RCIM_Cubic, leaveTangent=-122.6229, tangentMode=RCTM_User, time=0.5, value=55}
           {arriveTangent=-2.2589, interpMode=RCIM_Cubic, leaveTangent=-2.2589, tangentMode=RCTM_User, time=1, value=0}
          ]}}}
       CurveIndex=$User.NormalizedNotifyProgress
      )
     )
    )) {isSetParameters=true}
   &self:Reflect() {module=Reflect}
   &SubUVAnimation(End Frame=3, SubUV Animation Mode:ENiagara_SubUVLookupMode=::NewEnumerator2)
  ]
  particleUpdateScript [
   &ParticleState()
   /SetVariables(Particles.Size:NiagaraFloat=&FloatFromCurve(FloatCurve=DataInterface() {propertyValues {Curve {keys [
         {arriveTangent=5.5561, interpMode=RCIM_Cubic, leaveTangent=5.5561, tangentMode=RCTM_User, time=0, value=0}
         {arriveTangent=0.093, interpMode=RCIM_Cubic, leaveTangent=0.093, tangentMode=RCTM_User, time=0.2, value=1}
         <at 1 1.05>
        ]}}})) {isSetParameters=true}
   &ScaleSpriteSize(Scale Factor=&Vector2DFromFloat(Value=&Multiply_Float(
      A=$Particles.Size
      B=&SelectFloatFromArray(Direct Array Index=$Engine.QualityLevel, Float Selection Array=DataInterface() {propertyValues {FloatData [3,2,1,1,1,1]}})
     )))
   &ScaleMeshSize(Scale Factor=&VectorFromFloat(Value=$Particles.Size))
   &Color(
    Color=<rgba 5 5 5 1>
    Scale Alpha=&FloatFromCurve(FloatCurve=DataInterface() {propertyValues {Curve {keys [
         <at 0.2 1 RCIM_Cubic>
         {arriveTangent=-2.0008, interpMode=RCIM_Cubic, leaveTangent=-2.0008, tangentMode=RCTM_User, time=0.8, value=0.6}
         {arriveTangent=-2.5437, interpMode=RCIM_Cubic, leaveTangent=-2.5437, tangentMode=RCTM_User, time=1, value=0}
        ]}}})
   )
   &AccelerationForce(Acceleration=<40 0 10>)
   &CurlNoiseForce(Noise Frequency=75, Noise Strength=200, Pan Noise=false)
   &SolveForcesAndVelocity(Clamp Velocity=false, Limit Acceleration=false)
   /Game/Effects/Niagara/Modules/Orientation/FN_Tumble/v1/FN_Tumble(Axis 1 Spin Rate=7, Axis 2 Spin Rate=5)
   &AlignSpriteToMeshOrientation()
   &DynamicMaterialParameters(
    Index 0 Param 1=&AngleConversion(
     Angle:NiagaraFloat=&FloatFromCurve(
      FloatCurve=DataInterface() {propertyValues {Curve {keys [
          {arriveTangent=3.0149, interpMode=RCIM_Cubic, leaveTangent=3.0149, tangentMode=RCTM_User, time=0, value=0}
          {arriveTangent=-0.0084, interpMode=RCIM_Cubic, leaveTangent=-0.0084, tangentMode=RCTM_User, time=1, value=1}
         ]}}}
      Scale Curve=180
     )
     Angle Input:ENiagara_AngleInput=::NewEnumerator0
     Angle Output:ENiagara_AngleInput=::NewEnumerator1
    )
    Index 0 Param 2=3533.4648
    Index 0 Param 3=&RandomRangeFloat(Minimum=0.1, Maximum=0.9)
    Index 0 Param 4=&RandomRangeFloat(Minimum=0.1, Maximum=0.3)
    Module.Param0WriteEnabled=true
    Module.Param1WriteEnabled=true
    Module.Param2WriteEnabled=true
    Module.Param3WriteEnabled=true
   )
   /SetVariables(Particles.DynamicMaterialParameter1:Vector4f=&MakeVector4_FromLinearColor(LinearColor=&MakeLinearColorFromVectorAndFloat(
      "Vector (RGB)"=&Multiply_Vector(
       A=&Lerp_Vector(A=<0 0.2 1>, B=<0 0.5 1>, Alpha=&RandomRangeFloat(Minimum=0, Maximum=1))
       B=&VectorFromFloat(Value=&Lerp_Float(A=40, B=10, Alpha=$Particles.NormalizedAge))
      )
      "Float (Alpha)"=1
     ))) {isSetParameters=true}
  ]
  rendererClasses [&NiagaraSpriteRendererProperties, &NiagaraMeshRendererProperties]
  renderers [
   {rendererClass=&NiagaraSpriteRendererProperties}
   {rendererClass=&NiagaraMeshRendererProperties, rendererIndex=1}
  ]
 }
 FN_AttachedParticle_v3 {
  bEnabled=false
  emitterUpdateScript [
   &EmitterState()
   &SpawnBurst_Instantaneous(Spawn Count=4, Spawn Probability=HlslExpression() {hlslExpression="Emitter.LoopCount == 0"}, Use Spawn Probability=true)
  ]
  particleSpawnScript [
   &FN_SetSpawnGroup()
   /Niagara/Modules/Spawn/Initialization/InitializeParticle(
    Lifetime=10
    Mesh Scale=<2 2 2>
    Module.Write Color=true
    Module.Write Lifetime=true
    Module.Write Mass=true
    Module.Write Material Random=true
    Module.Write Position=false
    Module.Write Scale=true
    Module.Write SpriteRotation=false
    Module.Write SpriteSize=false
    Position=/Niagara/Functions/Localspace/SimulationPosition()
   )
  ]
  particleUpdateScript [
   &ParticleState()
   &SkeletalMeshLocation(
    Bone / Socket Index=$Particles.SpawnGroup
    Bone Sampling Mode=::NewEnumerator4
    Inherit Velocity=false
    Interpolate Mesh Sampling=false
    MaskSpawnGroup=false
    Skeletal Mesh=$User.SkeletalMesh
   ) {enabled=false}
   /SetVariables() {isSetParameters=true}
   /SetVariables(Particles.MeshOrientation=$Engine.Owner.Rotation) {isSetParameters=true}
  ]
  rendererClasses [&NiagaraMeshRendererProperties]
  renderers [{rendererClass=&NiagaraMeshRendererProperties, rendererIndex=0}]
 }
 FN_Core_Smoke_Sprite {
  simTarget=GPUComputeSim
  emitterUpdateScript [
   &EmitterState()
   &SpawnBurst_Instantaneous(Spawn Count=1, Use Spawn Probability=false) {enabled=false}
   &SpawnRate(
    SpawnRate=&Multiply_Float(
     A=7.5761
     B=&FloatFromCurve(
      FloatCurve=DataInterface() {propertyValues {Curve {keys [<at 0 0 RCIM_Cubic>,<at 0.1 1 RCIM_Cubic>,<at 1 0>]}}}
      CurveIndex=$User.NormalizedNotifyProgress
     )
    )
    Use Spawn Probability=false
   )
  ]
  particleSpawnScript [
   /SetVariables(Particles.InitSpriteRotRate=&RandomRangeFloat(Minimum=-2, Maximum=2)) {isSetParameters=true}
   /Niagara/Modules/Spawn/Initialization/V2/InitializeParticle(
    Color Maximum=<rgba 0.1255 0.1137 0.1059 1>
    Color Minimum=<rgba 0.2 0.1843 0.1686 1>
    Color Mode=::NewEnumerator2
    Lifetime Max=1.7
    Lifetime Min=1.2
    Lifetime Mode=::NewEnumerator1
    Sprite Rotation Angle=&RandomRangeFloat(Minimum=0, Maximum=360)
    Sprite Rotation Mode=::NewEnumerator0
    Sprite Size Mode=::NewEnumerator3
    Uniform Sprite Size=350
    UsePositionOffset=false
    Write Material Random=true
   )
   &CylinderLocation(
    Coordinate Space:ENiagaraCoordinateSpace=::World
    Cylinder Height=10.4054
    Cylinder Origin:Vector3f=/Niagara/DynamicInputs/Helpers/SimulationPosition()
    Cylinder Radius=&FloatFromCurve(
     FloatCurve=DataInterface() {propertyValues {Curve {keys [{arriveTangent=-0.0803, interpMode=RCIM_Cubic, leaveTangent=-0.0803, tangentMode=RCTM_User, time=0, value=0.475}, <at 1 1.025>]}}}
     CurveIndex=$User.NormalizedNotifyProgress
     Scale Curve=40
    )
    Hemisphere X=true
    Lathe Profile=DataInterface() {propertyValues {Curve {keys [<at 0 0.451 RCIM_Cubic>, {arriveTangent=0.049, interpMode=RCIM_Cubic, leaveTangent=0.049, time=0.5, value=1}, <at 1 0.5 RCIM_Cubic>]}}}
    MaskSpawnGroup=false
    Module.Enable Lathe Profile=false
    Module.Override Randomness=false
    Module.Surface Only=false
    Offset=&MakeVector(X=0, Y=0, Z=&Lerp_Float(A=120, B=20, Alpha=$User.NormalizedNotifyProgress))
   )
   /SetVariables(Particles.Position=&Add_Vector(
     A=&RotateVectorByQuaternion(
      Vector To Rotate=&Subtract_Vector(
       A=Linked() {linkedVariable {name=Particles.Position, type=NiagaraPosition}}
       B=Linked() {linkedVariable {name=Engine.Owner.Position, type=NiagaraPosition}}
      )
      Quaternion=&MakeQuatFromAxisAngle(
       Axis=$Engine.Owner.SystemYAxis
       AngleInDegrees=&FloatFromCurve(FloatCurve=DataInterface() {propertyValues {Curve {keys [<at 0 0.45>,<at 1 1>]}}}, CurveIndex=$User.NormalizedNotifyProgress, Scale Curve=90)
       NormalizeAxis=true
      )
     )
     B=Linked() {linkedVariable {name=Engine.Owner.Position, type=NiagaraPosition}}
    )) {enabled=false, isSetParameters=true}
   /SetVariables(Particles.Position=&Add_Vector(
     A=Linked() {linkedVariable {name=Particles.Position, type=NiagaraPosition}}
     B=&MakeVector(
      X=0
      Y=0
      Z=&FloatFromCurve(
       FloatCurve=DataInterface() {propertyValues {Curve {keys [
           {arriveTangent=-122.6229, interpMode=RCIM_Cubic, leaveTangent=-122.6229, tangentMode=RCTM_User, time=0, value=55}
           {arriveTangent=-2.2589, interpMode=RCIM_Cubic, leaveTangent=-2.2589, tangentMode=RCTM_User, time=1, value=0}
          ]}}}
       CurveIndex=$User.NormalizedNotifyProgress
      )
     )
    )) {isSetParameters=true}
   &self:Reflect() {module=Reflect}
  ]
  particleUpdateScript [
   &ParticleState()
   &SpriteRotationRate(Rotation Rate=&Multiply_Float(
     A=$Particles.InitSpriteRotRate
     B=&FloatFromCurve(
      FloatCurve=DataInterface() {propertyValues {Curve {keys [
          {arriveTangent=-4.0107, interpMode=RCIM_Cubic, leaveTangent=-4.0107, tangentMode=RCTM_User, time=-0.0046, value=1.0113}
          {arriveTangent=-0.2496, interpMode=RCIM_Cubic, leaveTangent=-0.2496, tangentMode=RCTM_User, time=0.5, value=0.16}
         ]}}}
      CurveIndex=$Emitter.Age
      Scale Curve=150
     )
    ))
   &ScaleColor(
    Scale Alpha=&FloatFromCurve(FloatCurve=DataInterface() {propertyValues {Curve {keys [<at 0 0.05 RCIM_Cubic>, {interpMode=RCIM_Cubic, tangentMode=RCTM_User, time=0.1, value=0.95}, <at 1 0>]}}})
    Scale RGB=&VectorFromFloat(Value=&FloatFromCurve(FloatCurve=DataInterface() {propertyValues {Curve {keys [{arriveTangent=1.8145, interpMode=RCIM_Cubic, leaveTangent=4.5795, tangentMode=RCTM_Break, time=0, value=0.4}, <at 0.3387 0.9565>, <at 1 1>]}}}))
    ScaleA=true
    ScaleRGB=true
   )
   &ScaleSpriteSize(Scale Factor=&Vector2DFromFloat(Value=&FloatFromCurve(
      FloatCurve=DataInterface() {propertyValues {Curve {keys [
          {arriveTangent=2.1256, interpMode=RCIM_Cubic, leaveTangent=2.1256, tangentMode=RCTM_User, time=0, value=0.15}
          {arriveTangent=1.0025, interpMode=RCIM_Cubic, leaveTangent=1.0025, tangentMode=RCTM_User, time=0.4, value=0.7}
          {arriveTangent=0.3209, interpMode=RCIM_Cubic, leaveTangent=0.3209, tangentMode=RCTM_User, time=1.0463, value=1}
         ]}}}
      Scale Curve=0.3
     )))
   &DynamicMaterialParameters(
    Index 0 Param 1=&Add_Float(
     A=$Particles.RandomFloat
     B=&FloatFromCurve(FloatCurve=DataInterface() {propertyValues {Curve {keys [
          {arriveTangent=2.7533, interpMode=RCIM_Cubic, leaveTangent=2.847, tangentMode=RCTM_Break, time=0, value=0}
          {arriveTangent=1.3581, interpMode=RCIM_Cubic, leaveTangent=1.3581, tangentMode=RCTM_User, time=0.2558, value=0.5125}
          {arriveTangent=0.0529, interpMode=RCIM_Cubic, leaveTangent=0.2021, tangentMode=RCTM_User, time=1, value=1}
         ]}}})
    )
    Index 0 Param 3=&FloatFromCurve(FloatCurve=DataInterface() {propertyValues {Curve {keys [{arriveTangent=-53.9199, time=0, value=0}, {leaveTangent=0.2723, time=0.25, value=0}]}}})
    Index 0 Param 4=&FloatFromCurve(FloatCurve=DataInterface() {propertyValues {Curve {keys [<at 0 1>,<at 1 0>]}}})
    Index 1 Param 1=&FloatFromCurve(FloatCurve=DataInterface() {propertyValues {Curve {keys [<at 0 0>, {arriveTangent=4.2257, interpMode=RCIM_Cubic, leaveTangent=4.2257, tangentMode=RCTM_User, time=0.25, value=0}]}}})
    Index 1 Param 2=&FloatFromCurve(FloatCurve=DataInterface() {propertyValues {Curve {keys [<at 0 0>]}}})
    Index 1 Param 3=&FloatFromCurve(FloatCurve=DataInterface() {propertyValues {Curve {keys [{interpMode=RCIM_Cubic, leaveTangent=6.3917, tangentMode=RCTM_Break, time=0, value=0}]}}})
    Index 1 Param 4=&FloatFromCurve(FloatCurve=DataInterface() {propertyValues {Curve {keys [{arriveTangent=-0.136, leaveTangent=-0.136, time=0, value=1}]}}})
    Index 2 Param 1=&FloatFromCurve(FloatCurve=DataInterface() {propertyValues {Curve {keys [
         {arriveTangent=2.646, interpMode=RCIM_Cubic, leaveTangent=2.646, tangentMode=RCTM_User, time=0, value=0}
         {leaveTangent=-0.0719, time=1, value=1}
        ]}}})
    Index 2 Param 2=&FloatFromCurve(
     FloatCurve=DataInterface() {propertyValues {Curve {keys [
         {arriveTangent=1.9056, time=0, value=0.2}
         {arriveTangent=0.0259, interpMode=RCIM_Cubic, leaveTangent=0.0259, tangentMode=RCTM_User, time=1, value=1.4}
        ]}}}
     Scale Curve=&RandomRangeFloat(Minimum=0.5, Maximum=1)
    )
    Index 2 Param 3=&FloatFromCurve(FloatCurve=DataInterface() {propertyValues {Curve {keys [
         {interpMode=RCIM_Cubic, leaveTangent=2.3064, tangentMode=RCTM_Break, time=0, value=1}
         {interpMode=RCIM_Cubic, leaveTangent=11.1233, tangentMode=RCTM_Break, time=1, value=2}
        ]}}})
    Index 2 Param 4=&FloatFromCurve(
     FloatCurve=DataInterface() {propertyValues {Curve {keys [
         {arriveTangent=-0.6867, interpMode=RCIM_Cubic, leaveTangent=-0.6867, tangentMode=RCTM_User, time=0, value=1}
         {interpMode=RCIM_Cubic, leaveTangent=-1.2956, tangentMode=RCTM_Break, time=1, value=0.75}
        ]}}}
     Scale Curve=1.6
    )
    Module.Param0WriteEnabled=true
    Module.Param0WriteEnabled1=true
    Module.Param0WriteEnabled2=true
    Module.Param1WriteEnabled=false
    Module.Param1WriteEnabled1=true
    Module.Param1WriteEnabled2=true
    Module.Param2WriteEnabled=true
    Module.Param2WriteEnabled1=true
    Module.Param2WriteEnabled2=true
    Module.Param3WriteEnabled=true
    Module.Param3WriteEnabled1=true
    Module.Param3WriteEnabled2=true
    Write Parameter Index 1=true
    Write Parameter Index 2=true
   )
   &AccelerationForce(Acceleration=<40 0 10>)
   &SolveForcesAndVelocity(Clamp Velocity=false, Limit Acceleration=false)
   &CameraOffset(Camera Offset Amount=40)
  ]
  rendererClasses [&NiagaraSpriteRendererProperties]
  renderers [
   {rendererClass=&NiagaraSpriteRendererProperties}
   {rendererClass=&NiagaraSpriteRendererProperties, rendererIndex=1}
  ]
 }
}
summary {
 systemName=NS_Emote_Faux
 emitters [
  {
   emitterName=Empty
   simTarget=GPUComputeSim
   rendererClasses [&NiagaraSpriteRendererProperties, &NiagaraMeshRendererProperties]
  }
  {bEnabled=false, emitterName=FN_AttachedParticle_v3, rendererClasses [&NiagaraMeshRendererProperties]}
  {emitterName=FN_Core_Smoke_Sprite, simTarget=GPUComputeSim, rendererClasses [&NiagaraSpriteRendererProperties]}
 ]
 userVariables [
  User.NormalizedNotifyProgress {type=NiagaraFloat, defaultValue=0.2, description=""}
  User.SkeletalMesh {
   type=NiagaraDataInterfaceSkeletalMesh
   defaultValue {
    struct=NiagaraExt_VariableValue_DataInterface
    value {dataInterface=&self:NiagaraDataInterfaceSkeletalMesh_0}
   }
  }
 ]
}
