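# Niagara effect for the Faux emote's head spray, built from two sprite emitters of short-lived, smoke-like puffs.
# Spray emits from a half-cylinder surface at a rate ramping with system loop age; Body_Spray follows after a delay from a lower box, ramping with its own age.
# Both push particles outward against drag, with curves driving growth, slowing spin, alpha and smoke material erosion and shading.
# User.NormalizedNotifyProgress is declared, but no listed module reads it.
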NiagaraSystem /BRCosmetics/Animation/Game/MainPlayer/Emotes/Faux/FX/NS_Emote_Faux_HeadSpray
dependencies {
 usedDataInterfaces [&NiagaraDataInterfaceCurve, &NiagaraDataInterfaceVectorCurve]
 usedDynamicInputs []
 usedModules [
  &SystemState
  &self:SystemSpawnScript.SystemScriptSource.SystemScriptGraph.NiagaraNodeAssignment_0.SetVariables_#1
 ]
}
emitters {
 Body_Spray {
  emitterUpdateScript [
   &EmitterState(Life Cycle Mode=::NewEnumerator1, Loop Behavior=::NewEnumerator1, Loop Delay=0.45, Loop Duration=0.5, UseLoopDelay=true)
   &SpawnBurst_Instantaneous(Spawn Count=&UniformRangedInt(Minimum=2, Maximum=5), Use Spawn Probability=false) {enabled=false}
   &SpawnRate(SpawnRate=&Multiply_Float(A=100, B=$Emitter.NormalizedLoopAge), Use Spawn Probability=false)
  ]
  particleSpawnScript [
   /Niagara/Modules/Spawn/Initialization/InitializeParticle(
    Lifetime=&UniformRangedFloat(Minimum=0.35, Maximum=0.45)
    Module.Write Color=true
    Module.Write Lifetime=true
    Module.Write Mass=true
    Module.Write Material Random=true
    Module.Write Position=false
    Module.Write Scale=false
    Module.Write SpriteRotation=true
    Module.Write SpriteSize=true
    Position=&SimulationPosition()
    Sprite Rotation=&NormalizedAngleToDegrees(NormalizedAngle=&UniformRangedFloat(Minimum=-15, Maximum=15))
    Sprite Size=&Vector2DFromFloat(Value=&UniformRangedFloat(Minimum=15, Maximum=30))
   )
   /SetVariables(Particles.InitSpriteRotRate=&UniformRangedFloat(Minimum=-3, Maximum=3)) {isSetParameters=true, module=SetVariables_#2}
   &AddVelocityInCone(
    Cone Angle=10
    Module.Use Velocity Falloff On Cone Axis=false
    Velocity Strength=&UniformRangedFloat(Minimum=150, Maximum=600)
   ) {enabled=false}
   &BoxLocation(Box Offset=<6 0 -15>, Box Size=<1 1 30>, MaskSpawnGroup=false, Module.Override Seed=false, Rotate Box=false, Surface Only=false)
   &BoxLocation(
    Box Offset=&VectorFromCurve(
     VectorCurve=DataInterface() {propertyValues {XCurve {keys [<at 0 10>,<at 1 10>]}, YCurve {keys [<at 0 1>,<at 1 1>]}, ZCurve {keys [<at 0 -5>,<at 1 -35>]}}}
     CurveIndex=$Emitter.NormalizedLoopAge
    )
    Box Size=<1 1 5>
    MaskSpawnGroup=false
    Module.Override Seed=false
    Rotate Box=false
    Surface Only=false
   ) {enabled=false, module=BoxLocation001}
   &AddVelocityFromPoint(Module.Use Falloff Distance=false, Velocity Strength=150)
  ]
  particleUpdateScript [
   &ParticleState()
   &SpriteRotationRate(Rotation Rate=&Multiply_Float(
     A=$Particles.InitSpriteRotRate
     B=&FloatFromCurve(FloatCurve=DataInterface() {propertyValues {Curve {keys [<at -0.0046 1.0113 RCIM_Cubic>,<at 1.0011 0.09 RCIM_Cubic>]}}})
    ))
   &Drag(Drag=3, Use Linear Drag=true, Use Rotational Drag=false)
   &SolveForcesAndVelocity(Clamp Velocity=false, Limit Acceleration=false)
   &ScaleSpriteSize(Scale Factor=&Vector2DFromFloat(Value=&FloatFromCurve(FloatCurve=DataInterface() {propertyValues {Curve {keys [
          {arriveTangent=0.4478, interpMode=RCIM_Cubic, leaveTangent=0.4478, tangentMode=RCTM_User, time=0.0013, value=0.0367}
          {arriveTangent=0.9127, interpMode=RCIM_Cubic, leaveTangent=0.9127, tangentMode=RCTM_User, time=0.5367, value=0.8831}
          {arriveTangent=3.2289, interpMode=RCIM_Cubic, leaveTangent=3.2289, tangentMode=RCTM_User, time=1.0008, value=1.6042}
         ]}}})))
   /Game/Effects/Niagara/Modules/Materials/MaterialSmoke(
    Erode:NiagaraFloat=&FloatFromCurve(FloatCurve=DataInterface() {propertyValues {
       bOptimizeLUT=false
       bUseLUT=false
       Curve {keys [
         {arriveTangent=-0.0652, interpMode=RCIM_Cubic, leaveTangent=-0.0652, tangentMode=RCTM_User, time=0.11, value=0.97}
         {arriveTangent=-5.7797, interpMode=RCIM_Cubic, leaveTangent=-5.7797, tangentMode=RCTM_User, time=0.39, value=0.66}
         {arriveTangent=-0.342, interpMode=RCIM_Cubic, leaveTangent=-0.342, tangentMode=RCTM_User, time=0.9954, value=-0.74}
        ]}
      }})
    Shading:NiagaraFloat=&FloatFromCurve(FloatCurve=DataInterface() {propertyValues {
       bOptimizeLUT=false
       bUseLUT=false
       Curve {keys [
         <at 0.07 0.5 RCIM_Cubic>
         {arriveTangent=1.2205, interpMode=RCIM_Cubic, leaveTangent=1.2205, time=0.17, value=1.02}
         {interpMode=RCIM_Cubic, tangentMode=RCTM_User, time=0.3895, value=0.89}
         <at 0.8 0 RCIM_Cubic>
        ]}
      }})
   )
   &CameraOffset(Camera Offset Amount=2.5)
   &DynamicMaterialParameters(
    Index 0 Param 1=&Add_Float(
     A=&RandomRangeFloat(Minimum=0, Maximum=1)
     B=&FloatFromCurve(FloatCurve=DataInterface() {propertyValues {Curve {keys [<at 0 1>,<at 1 0>]}}})
    )
    Index 0 Param 2=&FloatFromCurve(FloatCurve=DataInterface() {propertyValues {Curve {keys [
         {arriveTangent=0.8564, interpMode=RCIM_Cubic, leaveTangent=0.8564, tangentMode=RCTM_User, time=0.0046, value=-0.9753}
         {arriveTangent=4.2876, interpMode=RCIM_Cubic, leaveTangent=5.5268, tangentMode=RCTM_User, time=1, value=1}
        ]}}})
    Index 0 Param 3=&FloatFromCurve(FloatCurve=DataInterface() {propertyValues {Curve {keys [
         {arriveTangent=-22.7009, interpMode=RCIM_Cubic, leaveTangent=-22.7009, tangentMode=RCTM_User, time=-0.0012, value=10.368}
         {arriveTangent=-61.5499, interpMode=RCIM_Cubic, leaveTangent=-61.5499, tangentMode=RCTM_User, time=0.035, value=6.6513}
         <at 0.3531 0 RCIM_Cubic>
        ]}}})
    Index 0 Param 4=&FloatFromCurve(FloatCurve=DataInterface() {propertyValues {Curve {keys [{arriveTangent=-3.2497, interpMode=RCIM_Cubic, leaveTangent=-3.2497, tangentMode=RCTM_User, time=0, value=1}, <at 1 0 RCIM_Cubic>]}}})
    Module.Param0WriteEnabled=true
    Module.Param1WriteEnabled=true
    Module.Param2WriteEnabled=true
    Module.Param3WriteEnabled=true
   )
   &ScaleColor(
    Scale Alpha=&FloatFromCurve(FloatCurve=DataInterface() {propertyValues {Curve {keys [<at -0.003 -0.006>,<at 0.1806 0.5257>,<at 0.4581 0.2965>,<at 1 0>]}}})
    ScaleA=true
    ScaleRGB=true
   )
  ]
  rendererClasses [&NiagaraSpriteRendererProperties]
  renderers [{rendererClass=&NiagaraSpriteRendererProperties}]
 }
 Spray {
  emitterUpdateScript [
   &EmitterState()
   &SpawnBurst_Instantaneous(Spawn Count=&UniformRangedInt(Minimum=2, Maximum=5), Use Spawn Probability=false) {enabled=false}
   &SpawnRate(SpawnRate=&Multiply_Float(A=200, B=$System.NormalizedLoopAge), Use Spawn Probability=false)
  ]
  particleSpawnScript [
   &InitializeParticle(
    Lifetime=&UniformRangedFloat(Minimum=0.35, Maximum=0.45)
    Module.Write Color=true
    Module.Write Lifetime=true
    Module.Write Mass=true
    Module.Write Material Random=true
    Module.Write Position=false
    Module.Write Scale=false
    Module.Write SpriteRotation=true
    Module.Write SpriteSize=true
    Position=&SimulationPosition()
    Sprite Rotation=&NormalizedAngleToDegrees(NormalizedAngle=&UniformRangedFloat(Minimum=-15, Maximum=15))
    Sprite Size=&Vector2DFromFloat(Value=&UniformRangedFloat(Minimum=15, Maximum=30))
   )
   /SetVariables(Particles.InitSpriteRotRate=&UniformRangedFloat(Minimum=-3, Maximum=3)) {isSetParameters=true, module=SetVariables_#2}
   &AddVelocityInCone(
    Cone Angle=10
    Module.Use Velocity Falloff On Cone Axis=false
    Velocity Strength=&UniformRangedFloat(Minimum=150, Maximum=600)
   ) {enabled=false}
   &CylinderLocation(
    Cylinder Height=1
    Cylinder Radius=5
    Hemisphere Y=true
    MaskSpawnGroup=false
    Module.Enable Lathe Profile=false
    Module.Override Randomness=false
    Module.Surface Only=true
    Orientation Axis:ENiagaraOrientationAxis=::YAxis
    Surface Only Band Thickness=1
    Use Endcaps In Surface Only Mode=false
   )
   &AddVelocityFromPoint(Module.Use Falloff Distance=false)
  ]
  particleUpdateScript [
   &ParticleState()
   &SpriteRotationRate(Rotation Rate=&Multiply_Float(
     A=$Particles.InitSpriteRotRate
     B=&FloatFromCurve(FloatCurve=DataInterface() {propertyValues {Curve {keys [<at -0.0046 1.0113 RCIM_Cubic>,<at 1.0011 0.09 RCIM_Cubic>]}}})
    ))
   &Drag(Drag=3, Use Linear Drag=true, Use Rotational Drag=false)
   &SolveForcesAndVelocity(Clamp Velocity=false, Limit Acceleration=false)
   &ScaleSpriteSize(Scale Factor=&Vector2DFromFloat(Value=&FloatFromCurve(FloatCurve=DataInterface() {propertyValues {Curve {keys [
          {arriveTangent=0.4478, interpMode=RCIM_Cubic, leaveTangent=0.4478, tangentMode=RCTM_User, time=0.0013, value=0.0367}
          {arriveTangent=0.9127, interpMode=RCIM_Cubic, leaveTangent=0.9127, tangentMode=RCTM_User, time=0.5367, value=0.8831}
          {arriveTangent=3.2289, interpMode=RCIM_Cubic, leaveTangent=3.2289, tangentMode=RCTM_User, time=1.0008, value=1.6042}
         ]}}})))
   &MaterialSmoke(
    Erode:NiagaraFloat=&FloatFromCurve(FloatCurve=DataInterface() {propertyValues {
       bOptimizeLUT=false
       bUseLUT=false
       Curve {keys [
         {arriveTangent=-0.0652, interpMode=RCIM_Cubic, leaveTangent=-0.0652, tangentMode=RCTM_User, time=0.11, value=0.97}
         {arriveTangent=-5.7797, interpMode=RCIM_Cubic, leaveTangent=-5.7797, tangentMode=RCTM_User, time=0.39, value=0.66}
         {arriveTangent=-0.342, interpMode=RCIM_Cubic, leaveTangent=-0.342, tangentMode=RCTM_User, time=0.9954, value=-0.74}
        ]}
      }})
    Shading:NiagaraFloat=&FloatFromCurve(FloatCurve=DataInterface() {propertyValues {
       bOptimizeLUT=false
       bUseLUT=false
       Curve {keys [
         <at 0.07 0.5 RCIM_Cubic>
         {arriveTangent=1.2205, interpMode=RCIM_Cubic, leaveTangent=1.2205, time=0.17, value=1.02}
         {interpMode=RCIM_Cubic, tangentMode=RCTM_User, time=0.3895, value=0.89}
         <at 0.8 0 RCIM_Cubic>
        ]}
      }})
   )
   &CameraOffset(Camera Offset Amount=2.5)
   &DynamicMaterialParameters(
    Index 0 Param 1=&Add_Float(
     A=&RandomRangeFloat(Minimum=0, Maximum=1)
     B=&FloatFromCurve(FloatCurve=DataInterface() {propertyValues {Curve {keys [<at 0 1>,<at 1 0>]}}})
    )
    Index 0 Param 2=&FloatFromCurve(FloatCurve=DataInterface() {propertyValues {Curve {keys [
         {arriveTangent=0.8564, interpMode=RCIM_Cubic, leaveTangent=0.8564, tangentMode=RCTM_User, time=0.0046, value=-0.9753}
         {arriveTangent=4.2876, interpMode=RCIM_Cubic, leaveTangent=5.5268, tangentMode=RCTM_User, time=1, value=1}
        ]}}})
    Index 0 Param 3=&FloatFromCurve(FloatCurve=DataInterface() {propertyValues {Curve {keys [
         {arriveTangent=-22.7009, interpMode=RCIM_Cubic, leaveTangent=-22.7009, tangentMode=RCTM_User, time=-0.0012, value=10.368}
         {arriveTangent=-61.5499, interpMode=RCIM_Cubic, leaveTangent=-61.5499, tangentMode=RCTM_User, time=0.035, value=6.6513}
         <at 0.3531 0 RCIM_Cubic>
        ]}}})
    Index 0 Param 4=&FloatFromCurve(FloatCurve=DataInterface() {propertyValues {Curve {keys [{arriveTangent=-3.2497, interpMode=RCIM_Cubic, leaveTangent=-3.2497, tangentMode=RCTM_User, time=0, value=1}, <at 1 0 RCIM_Cubic>]}}})
    Module.Param0WriteEnabled=true
    Module.Param1WriteEnabled=true
    Module.Param2WriteEnabled=true
    Module.Param3WriteEnabled=true
   )
   &ScaleColor(
    Scale Alpha=&FloatFromCurve(FloatCurve=DataInterface() {propertyValues {Curve {keys [<at -0.003 -0.006>,<at 0.1806 0.5257>,<at 0.4581 0.2965>,<at 1 0>]}}})
    ScaleA=true
    ScaleRGB=true
   )
  ]
  rendererClasses [&NiagaraSpriteRendererProperties]
  renderers [{rendererClass=&NiagaraSpriteRendererProperties}]
 }
}
summary {
 systemName=NS_Emote_Faux_HeadSpray
 emitters [
  {emitterName=Spray, rendererClasses [&NiagaraSpriteRendererProperties]}
  {emitterName=Body_Spray, rendererClasses [&NiagaraSpriteRendererProperties]}
 ]
 userVariables [User.NormalizedNotifyProgress {type=NiagaraFloat, defaultValue=0.2, description=""}]
}
