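# Spray-on effect for the Faux emote, built from GPU emitters driven by User.NormalizedNotifyProgress.
# Spray meshes take their color from the Faux body texture, sampled along V as progress advances.
# A progress-curved height offset on the sampled skeletal-mesh position aims meshes and smoke along the line to the system origin.
# Smoke sprites drift from the origin or follow the sampled location, thinning late in the notify; several emitters are disabled.
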NiagaraSystem /BRCosmetics/Animation/Game/MainPlayer/Emotes/Faux/FX/NS_Emote_Faux_SprayOn
dependencies {
 usedDataInterfaces [/Script/Niagara.NiagaraDataInterfaceTexture, &NiagaraDataInterfaceCurve]
 usedDynamicInputs []
 usedModules [&SystemState]
}
emitters {
 Empty {
  bEnabled=false
  simTarget=GPUComputeSim
  emitterUpdateScript [
   &EmitterState()
   &SpawnBurst_Instantaneous(Spawn Count=2, Use Spawn Probability=false)
  ]
  particleSpawnScript [
   &InitializeParticle(
    Color Mode=::NewEnumerator0
    Lifetime=5
    Mesh Scale=&Multiply_VectorByFloat(Vector=<0.3 0.075 0.075>, Float=&RandomRangeFloat(Minimum=0.2, Maximum=0.7))
    Mesh Scale Mode=::NewEnumerator4
    UsePositionOffset=false
    Write Material Random=true
   )
   &InitialMeshOrientation(
    Rotation=&MakeVector(X=0, Y=0, Z=&RandomRangeFloat(Minimum=0, Maximum=405.538))
    Use Orientation Vector=false
    Use Rotation Vector=true
   )
  ]
  particleUpdateScript [
   &ParticleState()
   /Niagara/Modules/Update/Textures/SampleTexture(
    Texture=DataInterface() {propertyValues {
      Texture=/BRCosmetics/Characters/Player/Male/Medium/Bodies/M_MED_Faux/Textures/T_Faux_body_D
      TextureUserParameter {parameter {name=None, typeDefHandle {registeredTypeIndex=2}, varData []}}
     }}
    UV:Vector2f=&MakeVector2D(X=0.8, Y=&OneMinusFloat(Float=$User.NormalizedNotifyProgress))
   )
   &SkeletalMeshLocation(
    MaskSpawnGroup=false
    Mesh Sampling Type=::NewEnumerator5
    Orientation Sampling=::NewEnumerator1
    Position Sampling=::NewEnumerator2
    Skeletal Mesh=DataInterface() {propertyValues {
      DefaultMesh=null
      ExcludeBoneName=root
      OptionalComponentClass=null
      PreviewMesh=/SaveTheWorld/Characters/NPCs/Survivors/M_SML_Survivor_01/Meshes/SK_M_SML_Survivor_01_Preview
      SkinningMode=SkinOnTheFly
      SoftSourceActor=null
      SourceMode=Default
      UvSetIndex=0
      WholeMeshLOD=-1
      bExcludeBone=false
      bReadDeformedGeometry=true
      bRequireCurrentFrameData=true
      ComponentTags []
      FilteredBones []
      FilteredSockets [Root_Effect]
      MeshUserParameter {parameter {name=None, typeDefHandle {registeredTypeIndex=1}, varData []}}
      SamplingRegions []
     }}
    Socket Sampling Mode=::NewEnumerator5
    Use Sampled Mesh Attribute as Ribbon ID=false
   )
   /SetVariables(Particles.MeshOrientation=&FindQuatBetweenTwoVectors(
     From Vector=<-1 0 0>
     To Vector=&NormalizeVector(
      Vector To Normalize=&Subtract_Vector(
       A=Linked() {linkedVariable {name=Engine.Owner.Position, type=NiagaraPosition}}
       B=&Add_Vector(
        A=Linked() {linkedVariable {name=Output.SkeletalMeshLocation.SampledPositionWithoutOffset, type=NiagaraPosition}}
        B=&MakeVector(
         X=0
         Y=0
         Z=&FloatFromCurve(
          FloatCurve=DataInterface() {propertyValues {Curve {keys [
              <at 0.1 -0.3 RCIM_Cubic>
              {arriveTangent=1.6748, interpMode=RCIM_Cubic, leaveTangent=1.6748, tangentMode=RCTM_User, time=0.8, value=0.85}
              <at 1 0 RCIM_Cubic>
             ]}}}
          CurveIndex=$User.NormalizedNotifyProgress
          Scale Curve=150
         )
        )
       )
      )
      Safe Normalize=true
      Safe Fallback Threshold=0
      Safe Fallback Vector=<0 0 0.001>
     )
    )) {isSetParameters=true, module=SetVariables_#1}
   /SetVariables(Particles.Position=&Add_Vector(
     A=Linked() {linkedVariable {name=Engine.Owner.Position, type=NiagaraPosition}}
     B=&RotateVectorByQuaternion(Vector To Rotate=<5 0 0>, Quaternion=$Particles.MeshOrientation)
    )) {isSetParameters=true, module=SetVariables_#2}
   /SetVariables(Particles.XYScale:NiagaraFloat=&FloatFromCurve(
     FloatCurve=DataInterface() {propertyValues {Curve {keys [<at 0.1 0.9 RCIM_Cubic>, {arriveTangent=-0.0278, interpMode=RCIM_Cubic, leaveTangent=-0.0278, time=0.5, value=1.75}, <at 1 0.875 RCIM_Cubic>]}}}
     CurveIndex=$User.NormalizedNotifyProgress
    )) {isSetParameters=true, module=SetVariables_#3}
   &ScaleMeshSize(Scale Factor=&MakeVector(X=1, Y=$Particles.XYScale, Z=$Particles.XYScale))
   &DynamicMaterialParameters(
    Index 0 Param 1=&Add_Float(A=&Multiply_Float(A=$Particles.Age, B=1.0741), B=$Particles.MaterialRandom)
    Index 0 Param 2=&Add_Float(
     A=&Multiply_Float(A=$Particles.Age, B=1.115)
     B=&RandomRangeFloat(Minimum=0, Maximum=1)
    )
    Module.Param0WriteEnabled=true
    Module.Param1WriteEnabled=true
    Module.Param2WriteEnabled=true
    Module.Param3WriteEnabled=true
   )
   &Color(Color=&MultiplyLinearColor(Linear Color=$Output.SampleTexture.SampledColor, Scale Factor=<0.2 0.2 0.2 1>))
  ]
  rendererClasses [&NiagaraSpriteRendererProperties, &NiagaraMeshRendererProperties]
  renderers [
   {rendererClass=&NiagaraSpriteRendererProperties}
   {rendererClass=&NiagaraMeshRendererProperties, rendererIndex=1}
  ]
 }
 Empty001 {
  simTarget=GPUComputeSim
  emitterUpdateScript [
   &EmitterState()
   &SpawnBurst_Instantaneous(Spawn Count=1, Use Spawn Probability=false) {enabled=false}
   &SpawnRate(SpawnRate=45, Use Spawn Probability=false)
  ]
  particleSpawnScript [
   /SetVariables(Particles.XYSize:NiagaraFloat=&FloatFromCurve(
     FloatCurve=DataInterface() {propertyValues {Curve {keys [<at 0 0.1 RCIM_Cubic>, {arriveTangent=-0.05, interpMode=RCIM_Cubic, leaveTangent=-0.05, time=0.5, value=0.12}, <at 1 0.05 RCIM_Cubic>]}}}
     CurveIndex=$User.NormalizedNotifyProgress
     Scale Curve=0.75
    )) {isSetParameters=true}
   &SampleTexture(
    Texture=DataInterface() {propertyValues {
      Texture=&T_Faux_body_D
      TextureUserParameter {parameter {name=None, typeDefHandle {registeredTypeIndex=2}, varData []}}
     }}
    UV:Vector2f=&MakeVector2D(X=0.8, Y=&OneMinusFloat(Float=$User.NormalizedNotifyProgress))
   )
   &InitializeParticle(
    Color=$Output.SampleTexture.SampledColor
    Lifetime=&RandomRangeFloat(Minimum=0.3, Maximum=0.5)
    Mesh Scale=&Multiply_VectorByFloat(
     Vector=&MakeVector(X=$Particles.XYSize, Y=$Particles.XYSize, Z=0.4)
     Float=&RandomRangeFloat(Minimum=1, Maximum=2)
    )
    Mesh Scale Mode=::NewEnumerator4
    UsePositionOffset=false
    Write Material Random=true
   )
   &InitialMeshOrientation(
    Orientation Axis=<0 0 1>
    Orientation Vector=<0 0 -1>
    Rotation=&MakeVector(X=0, Y=0, Z=&RandomRangeFloat(Minimum=0, Maximum=405.538))
    Use Orientation Vector=true
    Use Rotation Vector=true
   )
  ]
  particleUpdateScript [
   &ParticleState()
   &SkeletalMeshLocation(
    MaskSpawnGroup=false
    Mesh Sampling Type=::NewEnumerator5
    Orientation Sampling=::NewEnumerator1
    Position Sampling=::NewEnumerator2
    Skeletal Mesh=$User.New Skeletal Mesh
    Socket Sampling Mode=::NewEnumerator5
    Use Sampled Mesh Attribute as Ribbon ID=false
   )
   /SetVariables() {isSetParameters=true, module=SetVariables_#2}
   /SetVariables(Particles.MeshOrientation=&MultiplyQuaternion(
     Quaternion A=&FindQuatBetweenTwoVectors(
      From Vector=<0 0 1>
      To Vector=&NormalizeVector(
       Vector To Normalize=&Subtract_Vector(
        A=Linked() {linkedVariable {name=Engine.Owner.Position, type=NiagaraPosition}}
        B=&Add_Vector(
         A=Linked() {linkedVariable {name=Output.SkeletalMeshLocation.SampledPositionWithoutOffset, type=NiagaraPosition}}
         B=&MakeVector(
          X=0
          Y=0
          Z=&FloatFromCurve(
           FloatCurve=DataInterface() {propertyValues {Curve {keys [
               <at 0.1 -0.3 RCIM_Cubic>
               {arriveTangent=1.6748, interpMode=RCIM_Cubic, leaveTangent=1.6748, tangentMode=RCTM_User, time=0.6, value=1.25}
               <at 1 0 RCIM_Cubic>
              ]}}}
           CurveIndex=$User.NormalizedNotifyProgress
           Scale Curve=118.3892
          )
         )
        )
       )
       Safe Normalize=true
       Safe Fallback Threshold=0
       Safe Fallback Vector=<0 0 0.001>
      )
     )
     Quaternion B=$Particles.Initial.MeshOrientation
    )) {isSetParameters=true, module=SetVariables_#1}
   /SetVariables(Particles.XYScale:NiagaraFloat=&FloatFromCurve(
     FloatCurve=DataInterface() {propertyValues {Curve {keys [<at 0.1 0.9 RCIM_Cubic>, {arriveTangent=-0.0278, interpMode=RCIM_Cubic, leaveTangent=-0.0278, time=0.5, value=1.3}, <at 1 0.875 RCIM_Cubic>]}}}
     CurveIndex=$User.NormalizedNotifyProgress
    )) {isSetParameters=true, module=SetVariables_#3}
   &ScaleMeshSize(Scale Factor=&MakeVector(X=1, Y=$Particles.XYScale, Z=$Particles.XYScale)) {enabled=false}
   &DynamicMaterialParameters(
    Index 0 Param 1=&Multiply_Float(A=$Particles.Age, B=0.6)
    Index 0 Param 2=&Multiply_Float(A=$Particles.Age, B=1.115)
    Index 1 Param 1=1
    Index 1 Param 2=&Multiply_Float(A=$Particles.Age, B=1.0233)
    Index 1 Param 3=0.75
    Index 1 Param 4=&Multiply_Float(A=$Particles.Age, B=0.362)
    Module.Param0WriteEnabled=true
    Module.Param0WriteEnabled1=true
    Module.Param1WriteEnabled=true
    Module.Param1WriteEnabled1=true
    Module.Param2WriteEnabled=true
    Module.Param2WriteEnabled1=true
    Module.Param3WriteEnabled=true
    Module.Param3WriteEnabled1=true
    Write Parameter Index 1=true
   )
   &Color(
    Color=$Particles.Initial.Color
    Scale Alpha=&FloatFromCurve(FloatCurve=DataInterface() {propertyValues {Curve {keys [<at 0 1>,<at 1 0>]}}})
    Scale Color=<2 2 2>
   )
   &ScaleMeshSize(Scale Factor=&MakeVector(
     X=1
     Y=1
     Z=&FloatFromCurve(FloatCurve=DataInterface() {propertyValues {Curve {keys [<at 0 1>,<at 1 1.95>]}}})
    )) {module=ScaleMeshSize001}
   &FN_Mesh_Attributes(
    Color=&RandomRangeLinearColor(Minimum=<rgba 7.239 14.8753 100 1>, Maximum=<rgba 9.5488 42.0303 100 1>)
    Create Color Attribute=true
    Create Scale Attribute=true
    Create Visibility Attribute=true
    Scale=&Multiply_Vector(A=$Particles.Scale, B=<1.1 1.1 0.5>)
    Write to Module Namespace=true
   )
  ]
  rendererClasses [&NiagaraSpriteRendererProperties, &NiagaraMeshRendererProperties]
  renderers [
   {rendererClass=&NiagaraSpriteRendererProperties}
   {rendererClass=&NiagaraMeshRendererProperties, rendererIndex=1}
   {rendererClass=&NiagaraMeshRendererProperties, rendererIndex=2}
  ]
 }
 FN_Core_Smoke_Sprite {
  bEnabled=false
  simTarget=GPUComputeSim
  emitterUpdateScript [
   &EmitterState()
   &SpawnBurst_Instantaneous(Spawn Count=1, Use Spawn Probability=false) {enabled=false}
   &SpawnRate(
    SpawnRate=&RemapRange(Input Value To Remap=$User.NormalizedNotifyProgress, Input Min=0.8, Input Max=1, Output Min=25, Output Max=0)
    Use Spawn Probability=false
   )
  ]
  particleSpawnScript [
   /SetVariables(Particles.InitSpriteRotRate=&RandomRangeFloat(Minimum=-2, Maximum=2)) {isSetParameters=true, module=SetVariables_#4}
   &InitializeParticle(
    Color Maximum=<rgba 0.1255 0.1137 0.1059 1>
    Color Minimum=<rgba 0.2 0.1843 0.1686 1>
    Color Mode=::NewEnumerator2
    Lifetime Max=1.7
    Lifetime Min=1.2
    Lifetime Mode=::NewEnumerator1
    Sprite Rotation Angle=&RandomRangeFloat(Minimum=0, Maximum=360)
    Sprite Rotation Mode=::NewEnumerator0
    Sprite Size Mode=::NewEnumerator3
    Uniform Sprite Size=350
    UsePositionOffset=false
    Write Material Random=true
   )
   &InitialMeshOrientation(
    Orientation Axis=<0 0 1>
    Orientation Vector=<0 0 -1>
    Rotation=&MakeVector(X=0, Y=0, Z=&RandomRangeFloat(Minimum=0, Maximum=405.538))
    Use Orientation Vector=true
    Use Rotation Vector=true
   )
  ]
  particleUpdateScript [
   &ParticleState()
   &SkeletalMeshLocation(
    MaskSpawnGroup=false
    Mesh Sampling Type=::NewEnumerator5
    Orientation Sampling=::NewEnumerator1
    Position Sampling=::NewEnumerator2
    Skeletal Mesh=$User.New Skeletal Mesh
    Socket Sampling Mode=::NewEnumerator5
    Use Sampled Mesh Attribute as Ribbon ID=false
   )
   /SetVariables(Particles.MeshOrientation=&MultiplyQuaternion(
     Quaternion A=&FindQuatBetweenTwoVectors(
      From Vector=<0 0 1>
      To Vector=&NormalizeVector(
       Vector To Normalize=&Subtract_Vector(
        A=Linked() {linkedVariable {name=Engine.Owner.Position, type=NiagaraPosition}}
        B=&Add_Vector(
         A=Linked() {linkedVariable {name=Output.SkeletalMeshLocation.SampledPositionWithoutOffset, type=NiagaraPosition}}
         B=&MakeVector(
          X=0
          Y=0
          Z=&FloatFromCurve(
           FloatCurve=DataInterface() {propertyValues {Curve {keys [
               <at 0.1 -0.3 RCIM_Cubic>
               {arriveTangent=1.6748, interpMode=RCIM_Cubic, leaveTangent=1.6748, tangentMode=RCTM_User, time=0.6, value=1.25}
               <at 1 0 RCIM_Cubic>
              ]}}}
           CurveIndex=$User.NormalizedNotifyProgress
           Scale Curve=118.3892
          )
         )
        )
       )
       Safe Normalize=true
       Safe Fallback Threshold=0
       Safe Fallback Vector=<0 0 0.001>
      )
     )
     Quaternion B=$Particles.Initial.MeshOrientation
    )) {isSetParameters=true, module=SetVariables_#5}
   /SetVariables(Particles.Position=&Lerp_Vector(
     A=Linked() {linkedVariable {name=Engine.Owner.Position, type=NiagaraPosition}}
     B=&Add_Vector(
      A=Linked() {linkedVariable {name=Engine.Owner.Position, type=NiagaraPosition}}
      B=&RotateVectorByQuaternion(Vector To Rotate=<0 0 50>, Quaternion=$Particles.MeshOrientation)
     )
     Alpha=$Particles.NormalizedAge
    )) {isSetParameters=true, module=SetVariables_#6}
   &SpriteRotationRate(Rotation Rate=&Multiply_Float(
     A=$Particles.InitSpriteRotRate
     B=&FloatFromCurve(
      FloatCurve=DataInterface() {propertyValues {Curve {keys [
          {arriveTangent=-4.0107, interpMode=RCIM_Cubic, leaveTangent=-4.0107, tangentMode=RCTM_User, time=-0.0046, value=1.0113}
          {arriveTangent=-0.2496, interpMode=RCIM_Cubic, leaveTangent=-0.2496, tangentMode=RCTM_User, time=0.5, value=0.16}
         ]}}}
      CurveIndex=$Emitter.Age
      Scale Curve=150
     )
    ))
   &ScaleColor(
    Scale Alpha=&FloatFromCurve(
     FloatCurve=DataInterface() {propertyValues {Curve {keys [{interpMode=RCIM_Cubic, tangentMode=RCTM_User, time=0, value=0.8}, <at 1 0>]}}}
     Scale Curve=&RemapRange(Input Value To Remap=$User.NormalizedNotifyProgress, Input Min=0.8, Input Max=1, Output Min=1, Output Max=0)
    )
    Scale RGB=&VectorFromFloat(Value=&FloatFromCurve(FloatCurve=DataInterface() {propertyValues {Curve {keys [{arriveTangent=1.8145, interpMode=RCIM_Cubic, leaveTangent=4.5795, tangentMode=RCTM_Break, time=0, value=0.4}, <at 0.3387 0.9565>, <at 1 1>]}}}))
    ScaleA=true
    ScaleRGB=true
   )
   &ScaleSpriteSize(Scale Factor=&Vector2DFromFloat(Value=&FloatFromCurve(
      FloatCurve=DataInterface() {propertyValues {Curve {keys [
          {arriveTangent=1.8507, interpMode=RCIM_Cubic, leaveTangent=1.8507, tangentMode=RCTM_User, time=0, value=0.15}
          {arriveTangent=1.0025, interpMode=RCIM_Cubic, leaveTangent=1.0025, tangentMode=RCTM_User, time=0.2, value=0.5}
          {arriveTangent=0.1638, interpMode=RCIM_Cubic, leaveTangent=0.1638, tangentMode=RCTM_User, time=1.0463, value=1}
         ]}}}
      Scale Curve=0.3
     )))
   &DynamicMaterialParameters(
    Index 0 Param 1=&Add_Float(
     A=$Particles.RandomFloat
     B=&FloatFromCurve(FloatCurve=DataInterface() {propertyValues {Curve {keys [
          {arriveTangent=2.7533, interpMode=RCIM_Cubic, leaveTangent=2.847, tangentMode=RCTM_Break, time=0, value=0}
          {arriveTangent=1.3581, interpMode=RCIM_Cubic, leaveTangent=1.3581, tangentMode=RCTM_User, time=0.2558, value=0.5125}
          {arriveTangent=0.0529, interpMode=RCIM_Cubic, leaveTangent=0.2021, tangentMode=RCTM_User, time=1, value=1}
         ]}}})
    )
    Index 0 Param 3=&FloatFromCurve(FloatCurve=DataInterface() {propertyValues {Curve {keys [{arriveTangent=-53.9199, time=0, value=0.05}, {leaveTangent=0.2723, time=0.2, value=0}]}}})
    Index 0 Param 4=&FloatFromCurve(FloatCurve=DataInterface() {propertyValues {Curve {keys [<at 0 1>,<at 1 0>]}}})
    Index 1 Param 1=&FloatFromCurve(FloatCurve=DataInterface() {propertyValues {Curve {keys [<at 0 0>, {arriveTangent=4.2257, interpMode=RCIM_Cubic, leaveTangent=4.2257, tangentMode=RCTM_User, time=0.25, value=0}]}}})
    Index 1 Param 2=&FloatFromCurve(FloatCurve=DataInterface() {propertyValues {Curve {keys [<at 0 0>]}}})
    Index 1 Param 3=&FloatFromCurve(FloatCurve=DataInterface() {propertyValues {Curve {keys [{interpMode=RCIM_Cubic, leaveTangent=6.3917, tangentMode=RCTM_Break, time=0, value=0}]}}})
    Index 1 Param 4=&FloatFromCurve(FloatCurve=DataInterface() {propertyValues {Curve {keys [{arriveTangent=-0.136, leaveTangent=-0.136, time=0, value=1}]}}})
    Index 2 Param 1=&FloatFromCurve(FloatCurve=DataInterface() {propertyValues {Curve {keys [
         {arriveTangent=2.646, interpMode=RCIM_Cubic, leaveTangent=2.646, tangentMode=RCTM_User, time=0, value=0}
         {leaveTangent=-0.0719, time=1, value=1}
        ]}}})
    Index 2 Param 2=&FloatFromCurve(
     FloatCurve=DataInterface() {propertyValues {Curve {keys [
         {arriveTangent=1.9056, time=0, value=0.2}
         {arriveTangent=0.0259, interpMode=RCIM_Cubic, leaveTangent=0.0259, tangentMode=RCTM_User, time=1, value=1.4}
        ]}}}
     Scale Curve=&RandomRangeFloat(Minimum=0.5, Maximum=1)
    )
    Index 2 Param 3=&FloatFromCurve(FloatCurve=DataInterface() {propertyValues {Curve {keys [
         {interpMode=RCIM_Cubic, leaveTangent=2.3064, tangentMode=RCTM_Break, time=0, value=1}
         {interpMode=RCIM_Cubic, leaveTangent=11.1233, tangentMode=RCTM_Break, time=1, value=2}
        ]}}})
    Index 2 Param 4=&FloatFromCurve(
     FloatCurve=DataInterface() {propertyValues {Curve {keys [
         {arriveTangent=-0.6867, interpMode=RCIM_Cubic, leaveTangent=-0.6867, tangentMode=RCTM_User, time=0, value=1}
         {interpMode=RCIM_Cubic, leaveTangent=-1.2956, tangentMode=RCTM_Break, time=1, value=0.75}
        ]}}}
     Scale Curve=1.6
    )
    Module.Param0WriteEnabled=true
    Module.Param0WriteEnabled1=true
    Module.Param0WriteEnabled2=true
    Module.Param1WriteEnabled=false
    Module.Param1WriteEnabled1=true
    Module.Param1WriteEnabled2=true
    Module.Param2WriteEnabled=true
    Module.Param2WriteEnabled1=true
    Module.Param2WriteEnabled2=true
    Module.Param3WriteEnabled=true
    Module.Param3WriteEnabled1=true
    Module.Param3WriteEnabled2=true
    Write Parameter Index 1=true
    Write Parameter Index 2=true
   )
   &CameraOffset(Camera Offset Amount=20)
  ]
  rendererClasses [&NiagaraSpriteRendererProperties]
  renderers [
   {rendererClass=&NiagaraSpriteRendererProperties}
   {rendererClass=&NiagaraSpriteRendererProperties, rendererIndex=1}
  ]
 }
 FN_Core_Smoke_Sprite001 {
  bEnabled=false
  simTarget=GPUComputeSim
  emitterUpdateScript [
   &EmitterState()
   &SpawnBurst_Instantaneous(Spawn Count=1, Use Spawn Probability=false) {enabled=false}
   &SpawnRate(
    SpawnRate=&RemapRange(Input Value To Remap=$User.NormalizedNotifyProgress, Input Min=0.8, Input Max=1, Output Min=15, Output Max=0)
    Use Spawn Probability=false
   )
  ]
  particleSpawnScript [
   /SetVariables(Particles.InitSpriteRotRate=&RandomRangeFloat(Minimum=-2, Maximum=2)) {isSetParameters=true, module=SetVariables_#4}
   &InitializeParticle(
    Color Maximum=<rgba 0.1255 0.1137 0.1059 1>
    Color Minimum=<rgba 0.2 0.1843 0.1686 1>
    Color Mode=::NewEnumerator2
    Lifetime Max=1.7
    Lifetime Min=1.2
    Lifetime Mode=::NewEnumerator1
    Sprite Rotation Angle=&RandomRangeFloat(Minimum=0, Maximum=360)
    Sprite Rotation Mode=::NewEnumerator0
    Sprite Size Mode=::NewEnumerator3
    Uniform Sprite Size=350
    UsePositionOffset=false
    Write Material Random=true
   )
   &InitialMeshOrientation(
    Orientation Axis=<0 0 1>
    Orientation Vector=<0 0 -1>
    Rotation=&MakeVector(X=0, Y=0, Z=&RandomRangeFloat(Minimum=0, Maximum=405.538))
    Use Orientation Vector=true
    Use Rotation Vector=true
   ) {enabled=false}
  ]
  particleUpdateScript [
   &ParticleState()
   &SkeletalMeshLocation(
    MaskSpawnGroup=false
    Mesh Sampling Type=::NewEnumerator5
    Orientation Sampling=::NewEnumerator1
    Position Sampling=::NewEnumerator2
    Skeletal Mesh=$User.New Skeletal Mesh
    Socket Sampling Mode=::NewEnumerator5
    Use Sampled Mesh Attribute as Ribbon ID=false
   )
   /SetVariables(Particles.MeshOrientation=&MultiplyQuaternion(
     Quaternion A=&FindQuatBetweenTwoVectors(
      From Vector=<0 0 1>
      To Vector=&NormalizeVector(
       Vector To Normalize=&Subtract_Vector(
        A=Linked() {linkedVariable {name=Engine.Owner.Position, type=NiagaraPosition}}
        B=&Add_Vector(
         A=Linked() {linkedVariable {name=Output.SkeletalMeshLocation.SampledPositionWithoutOffset, type=NiagaraPosition}}
         B=&MakeVector(
          X=0
          Y=0
          Z=&FloatFromCurve(
           FloatCurve=DataInterface() {propertyValues {Curve {keys [
               <at 0.1 -0.3 RCIM_Cubic>
               {arriveTangent=1.6748, interpMode=RCIM_Cubic, leaveTangent=1.6748, tangentMode=RCTM_User, time=0.6, value=1.25}
               <at 1 0 RCIM_Cubic>
              ]}}}
           CurveIndex=$User.NormalizedNotifyProgress
           Scale Curve=118.3892
          )
         )
        )
       )
       Safe Normalize=true
       Safe Fallback Threshold=0
       Safe Fallback Vector=<0 0 0.001>
      )
     )
     Quaternion B=$Particles.Initial.MeshOrientation
    )) {enabled=false, isSetParameters=true, module=SetVariables_#5}
   /SetVariables(Particles.Position=&Lerp_Vector(
     A=Linked() {linkedVariable {name=Engine.Owner.Position, type=NiagaraPosition}}
     B=&Add_Vector(
      A=Linked() {linkedVariable {name=Engine.Owner.Position, type=NiagaraPosition}}
      B=&RotateVectorByQuaternion(Vector To Rotate=<0 0 50>, Quaternion=$Particles.MeshOrientation)
     )
     Alpha=$Particles.NormalizedAge
    )) {enabled=false, isSetParameters=true, module=SetVariables_#6}
   &SpriteRotationRate(Rotation Rate=&Multiply_Float(
     A=$Particles.InitSpriteRotRate
     B=&FloatFromCurve(
      FloatCurve=DataInterface() {propertyValues {Curve {keys [
          {arriveTangent=-4.0107, interpMode=RCIM_Cubic, leaveTangent=-4.0107, tangentMode=RCTM_User, time=-0.0046, value=1.0113}
          {arriveTangent=-0.2496, interpMode=RCIM_Cubic, leaveTangent=-0.2496, tangentMode=RCTM_User, time=0.5, value=0.16}
         ]}}}
      CurveIndex=$Emitter.Age
      Scale Curve=150
     )
    ))
   &ScaleColor(
    Scale Alpha=&FloatFromCurve(
     FloatCurve=DataInterface() {propertyValues {Curve {keys [{interpMode=RCIM_Cubic, tangentMode=RCTM_User, time=0, value=0.8}, <at 1 0>]}}}
     Scale Curve=&RemapRange(Input Value To Remap=$User.NormalizedNotifyProgress, Input Min=0.8, Input Max=1, Output Min=1, Output Max=0)
    )
    Scale RGB=&VectorFromFloat(Value=&FloatFromCurve(FloatCurve=DataInterface() {propertyValues {Curve {keys [{arriveTangent=1.8145, interpMode=RCIM_Cubic, leaveTangent=4.5795, tangentMode=RCTM_Break, time=0, value=0.4}, <at 0.3387 0.9565>, <at 1 1>]}}}))
    ScaleA=true
    ScaleRGB=true
   )
   &ScaleSpriteSize(Scale Factor=&Vector2DFromFloat(Value=&FloatFromCurve(
      FloatCurve=DataInterface() {propertyValues {Curve {keys [
          {arriveTangent=1.8507, interpMode=RCIM_Cubic, leaveTangent=1.8507, tangentMode=RCTM_User, time=0, value=0.15}
          {arriveTangent=1.0025, interpMode=RCIM_Cubic, leaveTangent=1.0025, tangentMode=RCTM_User, time=0.2, value=0.5}
          {arriveTangent=0.1638, interpMode=RCIM_Cubic, leaveTangent=0.1638, tangentMode=RCTM_User, time=1.0463, value=1}
         ]}}}
      Scale Curve=0.3
     )))
   &DynamicMaterialParameters(
    Index 0 Param 1=&Add_Float(
     A=$Particles.RandomFloat
     B=&FloatFromCurve(FloatCurve=DataInterface() {propertyValues {Curve {keys [
          {arriveTangent=2.7533, interpMode=RCIM_Cubic, leaveTangent=2.847, tangentMode=RCTM_Break, time=0, value=0}
          {arriveTangent=1.3581, interpMode=RCIM_Cubic, leaveTangent=1.3581, tangentMode=RCTM_User, time=0.2558, value=0.5125}
          {arriveTangent=0.0529, interpMode=RCIM_Cubic, leaveTangent=0.2021, tangentMode=RCTM_User, time=1, value=1}
         ]}}})
    )
    Index 0 Param 3=&FloatFromCurve(FloatCurve=DataInterface() {propertyValues {Curve {keys [{arriveTangent=-53.9199, time=0, value=0.05}, {leaveTangent=0.2723, time=0.2, value=0}]}}})
    Index 0 Param 4=&FloatFromCurve(FloatCurve=DataInterface() {propertyValues {Curve {keys [<at 0 1>,<at 1 0>]}}})
    Index 1 Param 1=&FloatFromCurve(FloatCurve=DataInterface() {propertyValues {Curve {keys [<at 0 0>, {arriveTangent=4.2257, interpMode=RCIM_Cubic, leaveTangent=4.2257, tangentMode=RCTM_User, time=0.25, value=0}]}}})
    Index 1 Param 2=&FloatFromCurve(FloatCurve=DataInterface() {propertyValues {Curve {keys [<at 0 0>]}}})
    Index 1 Param 3=&FloatFromCurve(FloatCurve=DataInterface() {propertyValues {Curve {keys [{interpMode=RCIM_Cubic, leaveTangent=6.3917, tangentMode=RCTM_Break, time=0, value=0}]}}})
    Index 1 Param 4=&FloatFromCurve(FloatCurve=DataInterface() {propertyValues {Curve {keys [{arriveTangent=-0.136, leaveTangent=-0.136, time=0, value=1}]}}})
    Index 2 Param 1=&FloatFromCurve(FloatCurve=DataInterface() {propertyValues {Curve {keys [
         {arriveTangent=2.646, interpMode=RCIM_Cubic, leaveTangent=2.646, tangentMode=RCTM_User, time=0, value=0}
         {leaveTangent=-0.0719, time=1, value=1}
        ]}}})
    Index 2 Param 2=&FloatFromCurve(
     FloatCurve=DataInterface() {propertyValues {Curve {keys [
         {arriveTangent=1.9056, time=0, value=0.2}
         {arriveTangent=0.0259, interpMode=RCIM_Cubic, leaveTangent=0.0259, tangentMode=RCTM_User, time=1, value=1.4}
        ]}}}
     Scale Curve=&RandomRangeFloat(Minimum=0.5, Maximum=1)
    )
    Index 2 Param 3=&FloatFromCurve(FloatCurve=DataInterface() {propertyValues {Curve {keys [
         {interpMode=RCIM_Cubic, leaveTangent=2.3064, tangentMode=RCTM_Break, time=0, value=1}
         {interpMode=RCIM_Cubic, leaveTangent=11.1233, tangentMode=RCTM_Break, time=1, value=2}
        ]}}})
    Index 2 Param 4=&FloatFromCurve(
     FloatCurve=DataInterface() {propertyValues {Curve {keys [
         {arriveTangent=-0.6867, interpMode=RCIM_Cubic, leaveTangent=-0.6867, tangentMode=RCTM_User, time=0, value=1}
         {interpMode=RCIM_Cubic, leaveTangent=-1.2956, tangentMode=RCTM_Break, time=1, value=0.75}
        ]}}}
     Scale Curve=1.6
    )
    Module.Param0WriteEnabled=true
    Module.Param0WriteEnabled1=true
    Module.Param0WriteEnabled2=true
    Module.Param1WriteEnabled=false
    Module.Param1WriteEnabled1=true
    Module.Param1WriteEnabled2=true
    Module.Param2WriteEnabled=true
    Module.Param2WriteEnabled1=true
    Module.Param2WriteEnabled2=true
    Module.Param3WriteEnabled=true
    Module.Param3WriteEnabled1=true
    Module.Param3WriteEnabled2=true
    Write Parameter Index 1=true
    Write Parameter Index 2=true
   )
   &CameraOffset(Camera Offset Amount=20)
  ]
  rendererClasses [&NiagaraSpriteRendererProperties]
  renderers [
   {rendererClass=&NiagaraSpriteRendererProperties}
   {rendererClass=&NiagaraSpriteRendererProperties, rendererIndex=1}
  ]
 }
 FN_Core_Smoke_Sprite002 {
  simTarget=GPUComputeSim
  emitterUpdateScript [
   &EmitterState()
   &SpawnBurst_Instantaneous(Spawn Count=1, Use Spawn Probability=false) {enabled=false}
   &SpawnRate(
    SpawnRate=&RemapRange(Input Value To Remap=$User.NormalizedNotifyProgress, Input Min=0.8, Input Max=1, Output Min=20, Output Max=0)
    Use Spawn Probability=false
   )
  ]
  particleSpawnScript [
   /SetVariables(Particles.InitSpriteRotRate=&RandomRangeFloat(Minimum=-2, Maximum=2)) {isSetParameters=true, module=SetVariables_#4}
   &InitializeParticle(
    Color Maximum=<rgba 0.4 0.3625 0.3375 0.5>
    Color Minimum=<rgba 0.6 0.5529 0.5059 0.7>
    Color Mode=::NewEnumerator2
    Lifetime Max=1.5
    Lifetime Min=0.8
    Lifetime Mode=::NewEnumerator1
    Sprite Rotation Angle=&RandomRangeFloat(Minimum=0, Maximum=360)
    Sprite Rotation Mode=::NewEnumerator0
    Sprite Size Mode=::NewEnumerator3
    Uniform Sprite Size=350
    UsePositionOffset=false
    Write Material Random=true
   )
   &InitialMeshOrientation(
    Orientation Axis=<0 0 1>
    Orientation Vector=<0 0 -1>
    Rotation=&MakeVector(X=0, Y=0, Z=&RandomRangeFloat(Minimum=0, Maximum=405.538))
    Use Orientation Vector=true
    Use Rotation Vector=true
   )
  ]
  particleUpdateScript [
   &ParticleState()
   &SkeletalMeshLocation(
    MaskSpawnGroup=false
    Mesh Sampling Type=::NewEnumerator5
    Orientation Sampling=::NewEnumerator1
    Position Sampling=::NewEnumerator2
    Skeletal Mesh=$User.New Skeletal Mesh
    Socket Sampling Mode=::NewEnumerator5
    Use Sampled Mesh Attribute as Ribbon ID=false
   )
   /SetVariables(Particles.MeshOrientation=&MultiplyQuaternion(
     Quaternion A=&FindQuatBetweenTwoVectors(
      From Vector=<0 0 1>
      To Vector=&NormalizeVector(
       Vector To Normalize=&Subtract_Vector(
        A=Linked() {linkedVariable {name=Engine.Owner.Position, type=NiagaraPosition}}
        B=&Add_Vector(
         A=Linked() {linkedVariable {name=Output.SkeletalMeshLocation.SampledPositionWithoutOffset, type=NiagaraPosition}}
         B=&MakeVector(
          X=0
          Y=0
          Z=&FloatFromCurve(
           FloatCurve=DataInterface() {propertyValues {Curve {keys [
               <at 0.1 -0.3 RCIM_Cubic>
               {arriveTangent=1.6748, interpMode=RCIM_Cubic, leaveTangent=1.6748, tangentMode=RCTM_User, time=0.6, value=1.25}
               <at 1 0 RCIM_Cubic>
              ]}}}
           CurveIndex=$User.NormalizedNotifyProgress
           Scale Curve=118.3892
          )
         )
        )
       )
       Safe Normalize=true
       Safe Fallback Threshold=0
       Safe Fallback Vector=<0 0 0.001>
      )
     )
     Quaternion B=$Particles.Initial.MeshOrientation
    )) {isSetParameters=true, module=SetVariables_#5}
   /SetVariables(Particles.Position=&Lerp_Vector(
     A=Linked() {linkedVariable {name=Engine.Owner.Position, type=NiagaraPosition}}
     B=&Add_Vector(
      A=Linked() {linkedVariable {name=Engine.Owner.Position, type=NiagaraPosition}}
      B=&RotateVectorByQuaternion(Vector To Rotate=<0 0 50>, Quaternion=$Particles.MeshOrientation)
     )
     Alpha=$Particles.NormalizedAge
    )) {isSetParameters=true, module=SetVariables_#6}
   &SpriteRotationRate(Rotation Rate=&Multiply_Float(
     A=$Particles.InitSpriteRotRate
     B=&FloatFromCurve(
      FloatCurve=DataInterface() {propertyValues {Curve {keys [
          {arriveTangent=-4.0107, interpMode=RCIM_Cubic, leaveTangent=-4.0107, tangentMode=RCTM_User, time=-0.0046, value=1.0113}
          {arriveTangent=-0.2496, interpMode=RCIM_Cubic, leaveTangent=-0.2496, tangentMode=RCTM_User, time=0.5, value=0.16}
         ]}}}
      CurveIndex=$Emitter.Age
      Scale Curve=150
     )
    ))
   &ScaleColor(
    Scale Alpha=&FloatFromCurve(
     FloatCurve=DataInterface() {propertyValues {Curve {keys [{interpMode=RCIM_Cubic, tangentMode=RCTM_User, time=0, value=0.8}, <at 1 0>]}}}
     Scale Curve=&RemapRange(Input Value To Remap=$User.NormalizedNotifyProgress, Input Min=0.8, Input Max=1, Output Min=1, Output Max=0)
    )
    Scale RGB=&VectorFromFloat(Value=&FloatFromCurve(FloatCurve=DataInterface() {propertyValues {Curve {keys [{arriveTangent=1.8145, interpMode=RCIM_Cubic, leaveTangent=4.5795, tangentMode=RCTM_Break, time=0, value=0.4}, <at 0.3387 0.9565>, <at 1 1>]}}}))
    ScaleA=true
    ScaleRGB=true
   )
   &ScaleSpriteSize(Scale Factor=&Vector2DFromFloat(Value=&FloatFromCurve(
      FloatCurve=DataInterface() {propertyValues {Curve {keys [
          {arriveTangent=1.1634, interpMode=RCIM_Cubic, leaveTangent=1.1634, tangentMode=RCTM_User, time=0, value=0.15}
          {arriveTangent=1.0025, interpMode=RCIM_Cubic, leaveTangent=1.0025, tangentMode=RCTM_User, time=0.2714, value=0.4895}
          {arriveTangent=0.1638, interpMode=RCIM_Cubic, leaveTangent=0.1638, tangentMode=RCTM_User, time=1.0463, value=1}
         ]}}}
      Scale Curve=0.1
     )))
   &DynamicMaterialParameters(
    Index 0 Param 1=&Add_Float(
     A=$Particles.RandomFloat
     B=&FloatFromCurve(FloatCurve=DataInterface() {propertyValues {Curve {keys [
          {arriveTangent=2.7533, interpMode=RCIM_Cubic, leaveTangent=2.847, tangentMode=RCTM_Break, time=0, value=0}
          {arriveTangent=1.3581, interpMode=RCIM_Cubic, leaveTangent=1.3581, tangentMode=RCTM_User, time=0.2558, value=0.5125}
          {arriveTangent=0.0529, interpMode=RCIM_Cubic, leaveTangent=0.2021, tangentMode=RCTM_User, time=1, value=1}
         ]}}})
    )
    Index 0 Param 3=&FloatFromCurve(FloatCurve=DataInterface() {propertyValues {Curve {keys [{arriveTangent=-53.9199, time=0, value=0.05}, {leaveTangent=0.2723, time=0.2, value=0}]}}})
    Index 0 Param 4=&FloatFromCurve(FloatCurve=DataInterface() {propertyValues {Curve {keys [<at 0 1>,<at 1 0>]}}})
    Index 1 Param 1=&FloatFromCurve(FloatCurve=DataInterface() {propertyValues {Curve {keys [<at 0 0>, {arriveTangent=4.2257, interpMode=RCIM_Cubic, leaveTangent=4.2257, tangentMode=RCTM_User, time=0.25, value=0}]}}})
    Index 1 Param 2=&FloatFromCurve(FloatCurve=DataInterface() {propertyValues {Curve {keys [<at 0 0>]}}})
    Index 1 Param 3=&FloatFromCurve(FloatCurve=DataInterface() {propertyValues {Curve {keys [{interpMode=RCIM_Cubic, leaveTangent=6.3917, tangentMode=RCTM_Break, time=0, value=0}]}}})
    Index 1 Param 4=&FloatFromCurve(FloatCurve=DataInterface() {propertyValues {Curve {keys [{arriveTangent=-0.136, leaveTangent=-0.136, time=0, value=1}]}}})
    Index 2 Param 1=&FloatFromCurve(FloatCurve=DataInterface() {propertyValues {Curve {keys [
         {arriveTangent=2.646, interpMode=RCIM_Cubic, leaveTangent=2.646, tangentMode=RCTM_User, time=0, value=0}
         {leaveTangent=-0.0719, time=1, value=1}
        ]}}})
    Index 2 Param 2=&FloatFromCurve(
     FloatCurve=DataInterface() {propertyValues {Curve {keys [
         {arriveTangent=1.9056, time=0, value=0.2}
         {arriveTangent=0.0259, interpMode=RCIM_Cubic, leaveTangent=0.0259, tangentMode=RCTM_User, time=1, value=1.4}
        ]}}}
     Scale Curve=&RandomRangeFloat(Minimum=0.5, Maximum=1)
    )
    Index 2 Param 3=&FloatFromCurve(FloatCurve=DataInterface() {propertyValues {Curve {keys [
         {interpMode=RCIM_Cubic, leaveTangent=2.3064, tangentMode=RCTM_Break, time=0, value=1}
         {interpMode=RCIM_Cubic, leaveTangent=11.1233, tangentMode=RCTM_Break, time=1, value=2}
        ]}}})
    Index 2 Param 4=&FloatFromCurve(
     FloatCurve=DataInterface() {propertyValues {Curve {keys [
         {arriveTangent=-0.6867, interpMode=RCIM_Cubic, leaveTangent=-0.6867, tangentMode=RCTM_User, time=0, value=1}
         {interpMode=RCIM_Cubic, leaveTangent=-1.2956, tangentMode=RCTM_Break, time=1, value=0.75}
        ]}}}
     Scale Curve=1.6
    )
    Module.Param0WriteEnabled=true
    Module.Param0WriteEnabled1=true
    Module.Param0WriteEnabled2=true
    Module.Param1WriteEnabled=false
    Module.Param1WriteEnabled1=true
    Module.Param1WriteEnabled2=true
    Module.Param2WriteEnabled=true
    Module.Param2WriteEnabled1=true
    Module.Param2WriteEnabled2=true
    Module.Param3WriteEnabled=true
    Module.Param3WriteEnabled1=true
    Module.Param3WriteEnabled2=true
    Write Parameter Index 1=true
    Write Parameter Index 2=true
   )
   &CameraOffset(Camera Offset Amount=20)
  ]
  rendererClasses [&NiagaraSpriteRendererProperties]
  renderers [
   {rendererClass=&NiagaraSpriteRendererProperties}
   {rendererClass=&NiagaraSpriteRendererProperties, rendererIndex=1}
  ]
 }
 FN_Core_Smoke_Sprite003 {
  simTarget=GPUComputeSim
  emitterUpdateScript [
   &EmitterState()
   &SpawnBurst_Instantaneous(Spawn Count=1, Use Spawn Probability=false) {enabled=false}
   &SpawnRate(
    SpawnRate=&RemapRange(Input Value To Remap=$User.NormalizedNotifyProgress, Input Min=0.8, Input Max=1, Output Min=5, Output Max=0)
    Use Spawn Probability=false
   )
  ]
  particleSpawnScript [
   /SetVariables(Particles.InitSpriteRotRate=&RandomRangeFloat(Minimum=-2, Maximum=2)) {isSetParameters=true, module=SetVariables_#4}
   &InitializeParticle(
    Color Maximum=<rgba 0.4 0.3625 0.3375 0.5>
    Color Minimum=<rgba 0.6 0.5529 0.5059 0.7>
    Color Mode=::NewEnumerator2
    Lifetime Max=0.8
    Lifetime Min=0.5
    Lifetime Mode=::NewEnumerator1
    Sprite Rotation Angle=&RandomRangeFloat(Minimum=0, Maximum=360)
    Sprite Rotation Mode=::NewEnumerator0
    Sprite Size Mode=::NewEnumerator3
    Uniform Sprite Size=350
    UsePositionOffset=false
    Write Material Random=true
   )
   &InitialMeshOrientation(
    Orientation Axis=<0 0 1>
    Orientation Vector=<0 0 -1>
    Rotation=&MakeVector(X=0, Y=0, Z=&RandomRangeFloat(Minimum=0, Maximum=405.538))
    Use Orientation Vector=true
    Use Rotation Vector=true
   ) {enabled=false}
  ]
  particleUpdateScript [
   &ParticleState()
   &SkeletalMeshLocation(
    MaskSpawnGroup=false
    Mesh Sampling Type=::NewEnumerator5
    Orientation Sampling=::NewEnumerator1
    Position Sampling=::NewEnumerator2
    Skeletal Mesh=$User.New Skeletal Mesh
    Socket Sampling Mode=::NewEnumerator5
    Use Sampled Mesh Attribute as Ribbon ID=false
   )
   /SetVariables(Particles.MeshOrientation=&MultiplyQuaternion(
     Quaternion A=&FindQuatBetweenTwoVectors(
      From Vector=<0 0 1>
      To Vector=&NormalizeVector(
       Vector To Normalize=&Subtract_Vector(
        A=Linked() {linkedVariable {name=Engine.Owner.Position, type=NiagaraPosition}}
        B=&Add_Vector(
         A=Linked() {linkedVariable {name=Output.SkeletalMeshLocation.SampledPositionWithoutOffset, type=NiagaraPosition}}
         B=&MakeVector(
          X=0
          Y=0
          Z=&FloatFromCurve(
           FloatCurve=DataInterface() {propertyValues {Curve {keys [
               <at 0.1 -0.3 RCIM_Cubic>
               {arriveTangent=1.6748, interpMode=RCIM_Cubic, leaveTangent=1.6748, tangentMode=RCTM_User, time=0.6, value=1.25}
               <at 1 0 RCIM_Cubic>
              ]}}}
           CurveIndex=$User.NormalizedNotifyProgress
           Scale Curve=118.3892
          )
         )
        )
       )
       Safe Normalize=true
       Safe Fallback Threshold=0
       Safe Fallback Vector=<0 0 0.001>
      )
     )
     Quaternion B=$Particles.Initial.MeshOrientation
    )) {enabled=false, isSetParameters=true, module=SetVariables_#5}
   /SetVariables(Particles.Position=&Lerp_Vector(
     A=Linked() {linkedVariable {name=Engine.Owner.Position, type=NiagaraPosition}}
     B=&Add_Vector(
      A=Linked() {linkedVariable {name=Engine.Owner.Position, type=NiagaraPosition}}
      B=&RotateVectorByQuaternion(Vector To Rotate=<0 0 50>, Quaternion=$Particles.MeshOrientation)
     )
     Alpha=$Particles.NormalizedAge
    )) {enabled=false, isSetParameters=true, module=SetVariables_#6}
   &SpriteRotationRate(Rotation Rate=&Multiply_Float(
     A=$Particles.InitSpriteRotRate
     B=&FloatFromCurve(
      FloatCurve=DataInterface() {propertyValues {Curve {keys [
          {arriveTangent=-4.0107, interpMode=RCIM_Cubic, leaveTangent=-4.0107, tangentMode=RCTM_User, time=-0.0046, value=1.0113}
          {arriveTangent=-0.2496, interpMode=RCIM_Cubic, leaveTangent=-0.2496, tangentMode=RCTM_User, time=0.5, value=0.16}
         ]}}}
      CurveIndex=$Emitter.Age
      Scale Curve=150
     )
    ))
   &ScaleColor(
    Scale Alpha=&FloatFromCurve(
     FloatCurve=DataInterface() {propertyValues {Curve {keys [{interpMode=RCIM_Cubic, tangentMode=RCTM_User, time=0, value=0.8}, <at 1 0>]}}}
     Scale Curve=&RemapRange(Input Value To Remap=$User.NormalizedNotifyProgress, Input Min=0.8, Input Max=1, Output Min=1, Output Max=0)
    )
    Scale RGB=&VectorFromFloat(Value=&FloatFromCurve(FloatCurve=DataInterface() {propertyValues {Curve {keys [{arriveTangent=1.8145, interpMode=RCIM_Cubic, leaveTangent=4.5795, tangentMode=RCTM_Break, time=0, value=0.4}, <at 0.3387 0.9565>, <at 1 1>]}}}))
    ScaleA=true
    ScaleRGB=true
   )
   &ScaleSpriteSize(Scale Factor=&Vector2DFromFloat(Value=&FloatFromCurve(
      FloatCurve=DataInterface() {propertyValues {Curve {keys [
          {arriveTangent=1.4332, interpMode=RCIM_Cubic, leaveTangent=1.4332, tangentMode=RCTM_User, time=0, value=0.15}
          {arriveTangent=1.0025, interpMode=RCIM_Cubic, leaveTangent=1.0025, tangentMode=RCTM_User, time=0.2476, value=0.497}
          {arriveTangent=0.1638, interpMode=RCIM_Cubic, leaveTangent=0.1638, tangentMode=RCTM_User, time=1.0463, value=1}
         ]}}}
      Scale Curve=0.1
     )))
   &DynamicMaterialParameters(
    Index 0 Param 1=&Add_Float(
     A=$Particles.RandomFloat
     B=&FloatFromCurve(FloatCurve=DataInterface() {propertyValues {Curve {keys [
          {arriveTangent=2.7533, interpMode=RCIM_Cubic, leaveTangent=2.847, tangentMode=RCTM_Break, time=0, value=0}
          {arriveTangent=1.3581, interpMode=RCIM_Cubic, leaveTangent=1.3581, tangentMode=RCTM_User, time=0.2558, value=0.5125}
          {arriveTangent=0.0529, interpMode=RCIM_Cubic, leaveTangent=0.2021, tangentMode=RCTM_User, time=1, value=1}
         ]}}})
    )
    Index 0 Param 3=&FloatFromCurve(FloatCurve=DataInterface() {propertyValues {Curve {keys [{arriveTangent=-53.9199, time=0, value=0.05}, {leaveTangent=0.2723, time=0.2, value=0}]}}})
    Index 0 Param 4=&FloatFromCurve(FloatCurve=DataInterface() {propertyValues {Curve {keys [<at 0 1>,<at 1 0>]}}})
    Index 1 Param 1=&FloatFromCurve(FloatCurve=DataInterface() {propertyValues {Curve {keys [<at 0 0>, {arriveTangent=4.2257, interpMode=RCIM_Cubic, leaveTangent=4.2257, tangentMode=RCTM_User, time=0.25, value=0}]}}})
    Index 1 Param 2=&FloatFromCurve(FloatCurve=DataInterface() {propertyValues {Curve {keys [<at 0 0>]}}})
    Index 1 Param 3=&FloatFromCurve(FloatCurve=DataInterface() {propertyValues {Curve {keys [{interpMode=RCIM_Cubic, leaveTangent=6.3917, tangentMode=RCTM_Break, time=0, value=0}]}}})
    Index 1 Param 4=&FloatFromCurve(FloatCurve=DataInterface() {propertyValues {Curve {keys [{arriveTangent=-0.136, leaveTangent=-0.136, time=0, value=1}]}}})
    Index 2 Param 1=&FloatFromCurve(FloatCurve=DataInterface() {propertyValues {Curve {keys [
         {arriveTangent=2.646, interpMode=RCIM_Cubic, leaveTangent=2.646, tangentMode=RCTM_User, time=0, value=0}
         {leaveTangent=-0.0719, time=1, value=1}
        ]}}})
    Index 2 Param 2=&FloatFromCurve(
     FloatCurve=DataInterface() {propertyValues {Curve {keys [
         {arriveTangent=1.9056, time=0, value=0.2}
         {arriveTangent=0.0259, interpMode=RCIM_Cubic, leaveTangent=0.0259, tangentMode=RCTM_User, time=1, value=1.4}
        ]}}}
     Scale Curve=&RandomRangeFloat(Minimum=0.5, Maximum=1)
    )
    Index 2 Param 3=&FloatFromCurve(FloatCurve=DataInterface() {propertyValues {Curve {keys [
         {interpMode=RCIM_Cubic, leaveTangent=2.3064, tangentMode=RCTM_Break, time=0, value=1}
         {interpMode=RCIM_Cubic, leaveTangent=11.1233, tangentMode=RCTM_Break, time=1, value=2}
        ]}}})
    Index 2 Param 4=&FloatFromCurve(
     FloatCurve=DataInterface() {propertyValues {Curve {keys [
         {arriveTangent=-0.6867, interpMode=RCIM_Cubic, leaveTangent=-0.6867, tangentMode=RCTM_User, time=0, value=1}
         {interpMode=RCIM_Cubic, leaveTangent=-1.2956, tangentMode=RCTM_Break, time=1, value=0.75}
        ]}}}
     Scale Curve=1.6
    )
    Module.Param0WriteEnabled=true
    Module.Param0WriteEnabled1=true
    Module.Param0WriteEnabled2=true
    Module.Param1WriteEnabled=false
    Module.Param1WriteEnabled1=true
    Module.Param1WriteEnabled2=true
    Module.Param2WriteEnabled=true
    Module.Param2WriteEnabled1=true
    Module.Param2WriteEnabled2=true
    Module.Param3WriteEnabled=true
    Module.Param3WriteEnabled1=true
    Module.Param3WriteEnabled2=true
    Write Parameter Index 1=true
    Write Parameter Index 2=true
   )
   &CameraOffset(Camera Offset Amount=20)
  ]
  rendererClasses [&NiagaraSpriteRendererProperties]
  renderers [
   {rendererClass=&NiagaraSpriteRendererProperties}
   {rendererClass=&NiagaraSpriteRendererProperties, rendererIndex=1}
  ]
 }
}
summary {
 systemName=NS_Emote_Faux_SprayOn
 emitters [
  {
   bEnabled=false
   emitterName=Empty
   simTarget=GPUComputeSim
   rendererClasses [&NiagaraSpriteRendererProperties, &NiagaraMeshRendererProperties]
  }
  {
   emitterName=Empty001
   simTarget=GPUComputeSim
   rendererClasses [&NiagaraSpriteRendererProperties, &NiagaraMeshRendererProperties]
  }
  {bEnabled=false, emitterName=FN_Core_Smoke_Sprite, simTarget=GPUComputeSim, rendererClasses [&NiagaraSpriteRendererProperties]}
  {
   bEnabled=false
   emitterName=FN_Core_Smoke_Sprite001
   simTarget=GPUComputeSim
   rendererClasses [&NiagaraSpriteRendererProperties]
  }
  {emitterName=FN_Core_Smoke_Sprite002, simTarget=GPUComputeSim, rendererClasses [&NiagaraSpriteRendererProperties]}
  {emitterName=FN_Core_Smoke_Sprite003, simTarget=GPUComputeSim, rendererClasses [&NiagaraSpriteRendererProperties]}
 ]
 userVariables [
  User.New Skeletal Mesh {
   type=NiagaraDataInterfaceSkeletalMesh
   defaultValue {
    struct=NiagaraExt_VariableValue_DataInterface
    value {dataInterface=&self:NiagaraDataInterfaceSkeletalMesh_0}
   }
  }
  User.NormalizedNotifyProgress {type=NiagaraFloat, defaultValue=0, description=""}
 ]
}
